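# Post-attach animation Blueprint for the Swamp Knight Commando face accessory mesh, adding secondary physics motion to its hood.
# It copies the pose and curves from the attached parent mesh, then chains Anim Dynamics nodes for the top, right and left hood bones.
# Each simulates relative to the head with angular springs, damping, wind and per-bone angular limits, with a LOD threshold.
# The event graph holds only unconnected default nodes.
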AnimBlueprint /Game/Characters/Player/Male/Medium/Bodies/M_MED_Commando_SwampKnight/Meshes/Parts/Swamp_Knight_Commando_FaceAcc_Skeleton_AnimBlueprint
target_skeleton=&/Swamp_Knight_Commando_FaceAcc_Skeleton
graphs [
 AnimGraph {nodes [
   n1 = Anim Dynamics - Bone: dyn_hood_top(
    ComponentPose=Local To Component(LocalPose=Copy Pose From Mesh() {bCopyCurves=true, bUseAttachedParent=true})
    Alpha=1
    ExternalForce="0,0,0"
   ) {
    angularDampingOverride=0.7
    bAngularSpring=true
    bGravityOverrideInSimSpace=true
    bOverrideAngularDamping=true
    lODThreshold=1
    windScale=0.5
    alphaBoolBlend {blendOption=Linear}
    boundBone {boneName=dyn_hood_top}
    componentAppliedLinearAccClamp=<1000 1000 1000>
    componentLinearAccScale=<1 1 1>
    componentLinearVelScale=<1 1 1>
    physicsBodyDefinitions [{
      sphereCollisionRadius=0
      boundBone {boneName=dyn_hood_top}
      boxExtents=<10 20 5>
      constraintSetup {angularLimitsMax=<60 0 0>, angularLimitsMin=<-60 0 0>, angularTarget=<0 0 0>}
      localJointOffset=<0 0 -6>
     }]
    relativeSpaceBone {boneName=head}
   }
   n2 = Anim Dynamics - Bone: dyn_hood_r(ComponentPose=@n1, Alpha=1, ExternalForce="0,0,0") {
    angularDampingOverride=0.7
    bAngularSpring=true
    bGravityOverrideInSimSpace=true
    bOverrideAngularDamping=true
    lODThreshold=1
    windScale=0.5
    alphaBoolBlend {blendOption=Linear}
    boundBone {boneName=dyn_hood_r}
    componentAppliedLinearAccClamp=<1000 1000 1000>
    componentLinearAccScale=<1 1 1>
    componentLinearVelScale=<1 1 1>
    physicsBodyDefinitions [{
      sphereCollisionRadius=0
      boundBone {boneName=dyn_hood_r}
      boxExtents=<20 20 5>
      constraintSetup {angularLimitsMax=<0 0 60>, angularLimitsMin=<0 0 -60>, angularTarget=<0 0 0>}
      localJointOffset=<0 0 -6>
     }]
    relativeSpaceBone {boneName=head}
   }
   n3 = Anim Dynamics - Bone: dyn_hood_l(ComponentPose=@n2, Alpha=1, ExternalForce="0,0,0") {
    angularDampingOverride=0.7
    bAngularSpring=true
    bGravityOverrideInSimSpace=true
    bOverrideAngularDamping=true
    lODThreshold=1
    windScale=0.5
    alphaBoolBlend {blendOption=Linear}
    boundBone {boneName=dyn_hood_l}
    componentAppliedLinearAccClamp=<1000 1000 1000>
    componentLinearAccScale=<1 1 1>
    componentLinearVelScale=<1 1 1>
    physicsBodyDefinitions [{
      sphereCollisionRadius=0
      boundBone {boneName=dyn_hood_l}
      boxExtents=<20 20 5>
      constraintSetup {angularLimitsMax=<60 0 60>, angularLimitsMin=<-60 0 -60>, angularTarget=<0 0 0>}
      localJointOffset=<0 0 -6>
     }]
    relativeSpaceBone {boneName=head}
   }
   Output Pose(Result=Component To Local(ComponentPose=@n3)) {layerGroup=None}
  ]}
 EventGraph {nodes [
   TryGetPawnOwner()
   Event BlueprintUpdateAnimation()
  ]}
]
variables []
