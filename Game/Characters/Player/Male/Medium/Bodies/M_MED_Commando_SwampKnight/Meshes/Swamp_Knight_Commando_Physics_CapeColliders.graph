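# Physics asset for the Swamp Knight Commando player body, named as the collider set for its cape.
# Capsules cover the pelvis, spine, clavicles, neck, hip potion and first and fourth bones of eight skirt chains, with a sphere for the head.
# Arm and leg bodies carry no shapes.
# Constraints link bodies along the skeleton with all swing and twist motion free, so the listed angle limits do not restrict rotation.
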PhysicsAsset /Game/Characters/Player/Male/Medium/Bodies/M_MED_Commando_SwampKnight/Meshes/Swamp_Knight_Commando_Physics_CapeColliders
bodies {
 calf_l []
 calf_r []
 clavicle_l [{shapeType=Capsule, length=7.7636, radius=11.1427, center=<9.9642 -0.7741 -1.5923>, rotation=<rot 75.9352 -0.0001 175.5622>}]
 clavicle_r [{shapeType=Capsule, length=7.7636, radius=11.1427, center=<-9.9642 0.7741 1.5923>, rotation=<rot 75.9351 -0.0001 -4.4378>}]
 dyn_potion_MSTR [{shapeType=Capsule, length=8.25, radius=5.255, shapeName=dyn_potion_MSTR_capsule, center=<4.474 0 3.3795>, rotation=<rot 0 -90 85>}]
 dyn_skirt_back_01_l [{shapeType=Capsule, length=13.25, radius=13.5, shapeName=dyn_skirt_back_01_l_capsule, center=<0 8.7542 0>, rotation=<rot 90 0 0>}]
 dyn_skirt_back_01_r [{shapeType=Capsule, length=13.25, radius=13.5, shapeName=dyn_skirt_back_01_r_capsule, center=<0 8.7542 0>, rotation=<rot 90 0 0>}]
 dyn_skirt_back_04_l [{
   shapeType=Capsule
   length=22.75
   radius=13.5
   shapeName=dyn_skirt_back_04_l_capsule
   center=<-6.525 9.9473 -1.1097>
   rotation=<rot 90 -159.6559 -154.656>
  }]
 dyn_skirt_back_04_r [{
   shapeType=Capsule
   length=22.75
   radius=13.5
   shapeName=dyn_skirt_back_04_r_capsule
   center=<-6.525 9.9473 1.5381>
   rotation=<rot 90 16.8234 21.8233>
  }]
 dyn_skirt_front_01_l [{shapeType=Capsule, length=13.25, radius=13.5, shapeName=dyn_skirt_front_01_l_capsule, center=<0 8.7542 0>, rotation=<rot 90 0 0>}]
 dyn_skirt_front_01_r [{shapeType=Capsule, length=13.25, radius=13.5, shapeName=dyn_skirt_front_01_r_capsule, center=<0 8.7542 0>, rotation=<rot 90 0 0>}]
 dyn_skirt_front_04_l [{
   shapeType=Capsule
   length=22.75
   radius=13.5
   shapeName=dyn_skirt_front_04_l_capsule
   center=<-6.525 9.9473 0>
   rotation=<rot 90 -158.1827 -153.1827>
  }]
 dyn_skirt_front_04_r [{
   shapeType=Capsule
   length=22.75
   radius=13.5
   shapeName=dyn_skirt_front_04_r_capsule
   center=<-6.525 9.9473 0>
   rotation=<rot 90 20.4238 25.4238>
  }]
 dyn_skirt_mid_01_l [{shapeType=Capsule, length=13.25, radius=13.5, shapeName=dyn_skirt_mid_01_l_capsule, center=<0 8.7542 0>, rotation=<rot 90 0 0>}]
 dyn_skirt_mid_01_r [{shapeType=Capsule, length=13.25, radius=13.5, shapeName=dyn_skirt_mid_01_r_capsule, center=<0 8.7542 0>, rotation=<rot 90 0 0>}]
 dyn_skirt_mid_04_l [{
   shapeType=Capsule
   length=22.75
   radius=13.5
   shapeName=dyn_skirt_mid_04_l_capsule
   center=<-6.525 9.9473 0>
   rotation=<rot 90 -161.8383 -156.8384>
  }]
 dyn_skirt_mid_04_r [{
   shapeType=Capsule
   length=22.75
   radius=13.5
   shapeName=dyn_skirt_mid_04_r_capsule
   center=<-6.525 9.9473 0>
   rotation=<rot 90 16.3206 21.3206>
  }]
 dyn_skirt_side_01_l [{shapeType=Capsule, length=13.25, radius=13.5, shapeName=dyn_skirt_side_01_l_capsule, center=<0 8.7542 0>, rotation=<rot 90 0 0>}]
 dyn_skirt_side_01_r [{shapeType=Capsule, length=13.25, radius=13.5, shapeName=dyn_skirt_side_01_r_capsule, center=<0 8.7542 0>, rotation=<rot 90 0 0>}]
 dyn_skirt_side_04_l [{
   shapeType=Capsule
   length=22.75
   radius=13.5
   shapeName=dyn_skirt_side_04_l_capsule
   center=<-6.525 9.9473 0>
   rotation=<rot 90 -162.3339 -157.334>
  }]
 dyn_skirt_side_04_r [{
   shapeType=Capsule
   length=22.75
   radius=13.5
   shapeName=dyn_skirt_side_04_r_capsule
   center=<-6.525 9.9473 0>
   rotation=<rot 90 14.6951 19.6951>
  }]
 foot_l []
 foot_r []
 head [{shapeType=Sphere, radius=17.3125, center=<11.028 -0.174 0>}]
 lowerarm_l []
 lowerarm_r []
 neck_02 [{shapeType=Capsule, length=13.4375, radius=7.3175, center=<-0.2642 -0.3408 0>, rotation=<rot 90 -171.1193 -175.3978>}]
 pelvis [{shapeType=Capsule, length=19.6562, radius=14.7888, center=<-2.6215 -1.3695 0>}]
 spine_01 [{shapeType=Capsule, length=10.7187, radius=12.4138, center=<-2.6215 -0.7156 0>}]
 spine_02 [
  {shapeType=Capsule, length=10.5937, radius=10.945, center=<-2.6215 0.4993 0.0001>}
  {shapeType=Capsule, length=29.875, radius=11.1013, center=<2.5024 0.4992 3.9825>, rotation=<rot 90 0.0049 0.0049>}
  {shapeType=Capsule, length=29.875, radius=11.1013, center=<2.4704 0.4991 -3.9825>, rotation=<rot 90 -0.0396 -0.0396>}
 ]
 spine_03 [{shapeType=Capsule, length=9.1562, radius=10.7575, center=<-1.953 0.1364 0.0001>}]
 spine_04 [{shapeType=Capsule, length=14.6562, radius=13.6325, center=<2.1467 -1.2748 0.0002>}]
 spine_05 [
  {shapeType=Capsule, length=17.8437, radius=17.195, center=<-1.7514 2.4879 0>}
  {shapeType=Capsule, length=12.7812, radius=17.0625, center=<0.7875 3.1143 0>}
 ]
 thigh_l []
 thigh_r []
 upperarm_l []
 upperarm_r []
}
constraints [
 {
  bone1Name=spine_01
  bone2Name=pelvis
  swing1LimitDegrees=45
  swing1Motion=Free
  swing2LimitDegrees=45
  swing2Motion=Free
  twistLimitDegrees=45
  twistMotion=Free
 }
 {
  bone1Name=spine_02
  bone2Name=spine_01
  swing1LimitDegrees=45
  swing1Motion=Free
  swing2LimitDegrees=45
  swing2Motion=Free
  twistLimitDegrees=45
  twistMotion=Free
 }
 {
  bone1Name=spine_03
  bone2Name=spine_02
  swing1LimitDegrees=45
  swing1Motion=Free
  swing2LimitDegrees=45
  swing2Motion=Free
  twistLimitDegrees=45
  twistMotion=Free
 }
 {
  bone1Name=spine_04
  bone2Name=spine_03
  swing1LimitDegrees=45
  swing1Motion=Free
  swing2LimitDegrees=45
  swing2Motion=Free
  twistLimitDegrees=45
  twistMotion=Free
 }
 {
  bone1Name=spine_05
  bone2Name=spine_04
  swing1LimitDegrees=45
  swing1Motion=Free
  swing2LimitDegrees=45
  swing2Motion=Free
  twistLimitDegrees=45
  twistMotion=Free
 }
 {
  bone1Name=clavicle_l
  bone2Name=spine_05
  swing1LimitDegrees=45
  swing1Motion=Free
  swing2LimitDegrees=45
  swing2Motion=Free
  twistLimitDegrees=45
  twistMotion=Free
 }
 {
  bone1Name=upperarm_l
  bone2Name=clavicle_l
  swing1LimitDegrees=45
  swing1Motion=Free
  swing2LimitDegrees=45
  swing2Motion=Free
  twistLimitDegrees=45
  twistMotion=Free
 }
 {
  bone1Name=lowerarm_l
  bone2Name=upperarm_l
  swing1LimitDegrees=45
  swing1Motion=Free
  swing2LimitDegrees=45
  swing2Motion=Free
  twistLimitDegrees=45
  twistMotion=Free
 }
 {
  bone1Name=clavicle_r
  bone2Name=spine_05
  swing1LimitDegrees=45
  swing1Motion=Free
  swing2LimitDegrees=45
  swing2Motion=Free
  twistLimitDegrees=45
  twistMotion=Free
 }
 {
  bone1Name=upperarm_r
  bone2Name=clavicle_r
  swing1LimitDegrees=45
  swing1Motion=Free
  swing2LimitDegrees=45
  swing2Motion=Free
  twistLimitDegrees=45
  twistMotion=Free
 }
 {
  bone1Name=lowerarm_r
  bone2Name=upperarm_r
  swing1LimitDegrees=45
  swing1Motion=Free
  swing2LimitDegrees=45
  swing2Motion=Free
  twistLimitDegrees=45
  twistMotion=Free
 }
 {
  bone1Name=neck_02
  bone2Name=spine_05
  swing1LimitDegrees=45
  swing1Motion=Free
  swing2LimitDegrees=45
  swing2Motion=Free
  twistLimitDegrees=45
  twistMotion=Free
 }
 {
  bone1Name=head
  bone2Name=neck_02
  swing1LimitDegrees=45
  swing1Motion=Free
  swing2LimitDegrees=45
  swing2Motion=Free
  twistLimitDegrees=45
  twistMotion=Free
 }
 {
  bone1Name=thigh_l
  bone2Name=pelvis
  swing1LimitDegrees=45
  swing1Motion=Free
  swing2LimitDegrees=45
  swing2Motion=Free
  twistLimitDegrees=45
  twistMotion=Free
 }
 {
  bone1Name=calf_l
  bone2Name=thigh_l
  swing1LimitDegrees=45
  swing1Motion=Free
  swing2LimitDegrees=45
  swing2Motion=Free
  twistLimitDegrees=45
  twistMotion=Free
 }
 {
  bone1Name=foot_l
  bone2Name=calf_l
  swing1LimitDegrees=45
  swing1Motion=Free
  swing2LimitDegrees=45
  swing2Motion=Free
  twistLimitDegrees=45
  twistMotion=Free
 }
 {
  bone1Name=thigh_r
  bone2Name=pelvis
  swing1LimitDegrees=45
  swing1Motion=Free
  swing2LimitDegrees=45
  swing2Motion=Free
  twistLimitDegrees=45
  twistMotion=Free
 }
 {
  bone1Name=calf_r
  bone2Name=thigh_r
  swing1LimitDegrees=45
  swing1Motion=Free
  swing2LimitDegrees=45
  swing2Motion=Free
  twistLimitDegrees=45
  twistMotion=Free
 }
 {
  bone1Name=foot_r
  bone2Name=calf_r
  swing1LimitDegrees=45
  swing1Motion=Free
  swing2LimitDegrees=45
  swing2Motion=Free
  twistLimitDegrees=45
  twistMotion=Free
 }
 {
  bone1Name=dyn_skirt_back_04_l
  bone2Name=dyn_skirt_back_01_l
  swing1LimitDegrees=45
  swing1Motion=Free
  swing2LimitDegrees=45
  swing2Motion=Free
  twistLimitDegrees=45
  twistMotion=Free
 }
 {
  bone1Name=dyn_skirt_back_04_r
  bone2Name=dyn_skirt_back_01_r
  swing1LimitDegrees=45
  swing1Motion=Free
  swing2LimitDegrees=45
  swing2Motion=Free
  twistLimitDegrees=45
  twistMotion=Free
 }
 {
  bone1Name=dyn_skirt_front_04_l
  bone2Name=dyn_skirt_front_01_l
  swing1LimitDegrees=45
  swing1Motion=Free
  swing2LimitDegrees=45
  swing2Motion=Free
  twistLimitDegrees=45
  twistMotion=Free
 }
 {
  bone1Name=dyn_skirt_front_04_r
  bone2Name=dyn_skirt_front_01_r
  swing1LimitDegrees=45
  swing1Motion=Free
  swing2LimitDegrees=45
  swing2Motion=Free
  twistLimitDegrees=45
  twistMotion=Free
 }
 {
  bone1Name=dyn_skirt_mid_04_l
  bone2Name=dyn_skirt_mid_01_l
  swing1LimitDegrees=45
  swing1Motion=Free
  swing2LimitDegrees=45
  swing2Motion=Free
  twistLimitDegrees=45
  twistMotion=Free
 }
 {
  bone1Name=dyn_skirt_mid_04_r
  bone2Name=dyn_skirt_mid_01_r
  swing1LimitDegrees=45
  swing1Motion=Free
  swing2LimitDegrees=45
  swing2Motion=Free
  twistLimitDegrees=45
  twistMotion=Free
 }
 {
  bone1Name=dyn_skirt_side_04_l
  bone2Name=dyn_skirt_side_01_l
  swing1LimitDegrees=45
  swing1Motion=Free
  swing2LimitDegrees=45
  swing2Motion=Free
  twistLimitDegrees=45
  twistMotion=Free
 }
 {
  bone1Name=dyn_skirt_side_04_r
  bone2Name=dyn_skirt_side_01_r
  swing1LimitDegrees=45
  swing1Motion=Free
  swing2LimitDegrees=45
  swing2Motion=Free
  twistLimitDegrees=45
  twistMotion=Free
 }
 {
  bone1Name=dyn_potion_MSTR
  bone2Name=pelvis
  swing1LimitDegrees=45
  swing1Motion=Free
  swing2LimitDegrees=45
  swing2Motion=Free
  twistLimitDegrees=45
  twistMotion=Free
 }
 {
  bone1Name=dyn_skirt_back_01_l
  bone2Name=pelvis
  swing1LimitDegrees=45
  swing1Motion=Free
  swing2LimitDegrees=45
  swing2Motion=Free
  twistLimitDegrees=45
  twistMotion=Free
 }
 {
  bone1Name=dyn_skirt_back_01_r
  bone2Name=pelvis
  swing1LimitDegrees=45
  swing1Motion=Free
  swing2LimitDegrees=45
  swing2Motion=Free
  twistLimitDegrees=45
  twistMotion=Free
 }
 {
  bone1Name=dyn_skirt_front_01_l
  bone2Name=pelvis
  swing1LimitDegrees=45
  swing1Motion=Free
  swing2LimitDegrees=45
  swing2Motion=Free
  twistLimitDegrees=45
  twistMotion=Free
 }
 {
  bone1Name=dyn_skirt_front_01_r
  bone2Name=pelvis
  swing1LimitDegrees=45
  swing1Motion=Free
  swing2LimitDegrees=45
  swing2Motion=Free
  twistLimitDegrees=45
  twistMotion=Free
 }
 {
  bone1Name=dyn_skirt_mid_01_l
  bone2Name=pelvis
  swing1LimitDegrees=45
  swing1Motion=Free
  swing2LimitDegrees=45
  swing2Motion=Free
  twistLimitDegrees=45
  twistMotion=Free
 }
 {
  bone1Name=dyn_skirt_mid_01_r
  bone2Name=pelvis
  swing1LimitDegrees=45
  swing1Motion=Free
  swing2LimitDegrees=45
  swing2Motion=Free
  twistLimitDegrees=45
  twistMotion=Free
 }
 {
  bone1Name=dyn_skirt_side_01_l
  bone2Name=pelvis
  swing1LimitDegrees=45
  swing1Motion=Free
  swing2LimitDegrees=45
  swing2Motion=Free
  twistLimitDegrees=45
  twistMotion=Free
 }
 {
  bone1Name=dyn_skirt_side_01_r
  bone2Name=pelvis
  swing1LimitDegrees=45
  swing1Motion=Free
  swing2LimitDegrees=45
  swing2Motion=Free
  twistLimitDegrees=45
  twistMotion=Free
 }
]
