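# Physics asset for the Swamp Knight male commando's skirt and a dangling bag, built entirely from capsule bodies.
# Eight four-bone skirt chains (front, side, mid and back on each side) hang from the pelvis, with limited swing and mostly tight twist.
# The bag is a single limited body on the pelvis.
# Root, pelvis, thigh, thigh-twist and calf capsules give the body's shape, linked by joints with free swing and twist.
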PhysicsAsset /Game/Characters/Player/Male/Medium/Bodies/M_MED_Commando_SwampKnight/Meshes/M_MED_Swamp_Knight_Skirt_Physics
bodies {
 calf_l [{
   shapeType=Capsule
   length=25.5545
   radius=10.262
   shapeName=calf_l_capsule
   center=<-19.4481 -2.2552 -0.9122>
   rotation=<rot 82.6131 -59.5065 120.3349>
  }]
 calf_r [{
   shapeType=Capsule
   length=23.9494
   radius=11.8026
   shapeName=calf_r_capsule
   center=<20.3517 4.1709 1.1817>
   rotation=<rot 77.7706 79.9487 -93.8064>
  }]
 dyn_bag [{shapeType=Capsule, length=8.9349, radius=6.9232, shapeName=dyn_bag_capsule, center=<6 1.3 5.292>, rotation=<rot 0 93.5439 -73.9883>}]
 dyn_skirt_back_01_l [{shapeType=Capsule, length=14.75, radius=7.75, shapeName=dyn_skirt_back_01_r_capsule, center=<5.6 -4.6865 0>}]
 dyn_skirt_back_01_r [{shapeType=Capsule, length=14.75, radius=7.75, shapeName=dyn_skirt_back_01_r_capsule, center=<5.6 -5.8318 0>}]
 dyn_skirt_back_02_l [{shapeType=Capsule, length=14.75, radius=7.75, shapeName=dyn_skirt_back_02_r_capsule, center=<5.6 -2.5634 0>}]
 dyn_skirt_back_02_r [{shapeType=Capsule, length=14.75, radius=7.75, shapeName=dyn_skirt_back_02_r_capsule, center=<5.6 -2.5634 0>}]
 dyn_skirt_back_03_l [{shapeType=Capsule, length=24.5, radius=7.75, shapeName=dyn_skirt_back_03_r_capsule, center=<6.6 -0.4393 0>}]
 dyn_skirt_back_03_r [{shapeType=Capsule, length=24.5, radius=7.75, shapeName=dyn_skirt_back_03_r_capsule, center=<6.6 -0.4393 0>}]
 dyn_skirt_back_04_l [{shapeType=Capsule, length=16.25, radius=8.75, shapeName=dyn_skirt_back_04_r_capsule, center=<7 -1.1783 0>}]
 dyn_skirt_back_04_r [{shapeType=Capsule, length=16.25, radius=8.75, shapeName=dyn_skirt_back_04_r_capsule, center=<7 -1.1783 0>}]
 dyn_skirt_front_01_l [{shapeType=Capsule, length=15, radius=6.75, shapeName=dyn_skirt_front_01_l_capsule, center=<3.3 -1.4 0>}]
 dyn_skirt_front_01_r [{shapeType=Capsule, length=15, radius=6.75, shapeName=dyn_skirt_front_01_r_capsule, center=<3.3 -3.3 0>}]
 dyn_skirt_front_02_l [{shapeType=Capsule, length=15.25, radius=7.25, shapeName=dyn_skirt_front_02_l_capsule, center=<5.6 -3.8 0>}]
 dyn_skirt_front_02_r [{
   shapeType=Capsule
   length=10
   radius=7.5
   shapeName=dyn_skirt_front_02_r_capsule
   center=<5.6 -5.9995 0>
   rotation=<rot 10.0001 179.9999 179.9999>
  }]
 dyn_skirt_front_03_l [{shapeType=Capsule, length=14.75, radius=7.75, shapeName=dyn_skirt_front_03_l_capsule, center=<5.6 -3.3 0>}]
 dyn_skirt_front_03_r [{
   shapeType=Capsule
   length=13.5
   radius=7.75
   shapeName=dyn_skirt_front_03_r_capsule
   center=<8.319 -7.2503 -0.7451>
   rotation=<rot 5 179.9999 180>
  }]
 dyn_skirt_front_04_l [{shapeType=Capsule, length=19.75, radius=9, shapeName=dyn_skirt_front_04_l_capsule, center=<6.3 -6 0>, rotation=<rot 5 0 0>}]
 dyn_skirt_front_04_r [{
   shapeType=Capsule
   length=19.25
   radius=9.5
   shapeName=dyn_skirt_front_04_r_capsule
   center=<7.8522 -6.9109 -1.6545>
   rotation=<rot -9.9809 -0.442 5.0577>
  }]
 dyn_skirt_mid_01_l [{shapeType=Capsule, length=15.25, radius=7.25, shapeName=dyn_skirt_mid_01_l_capsule, center=<5.6 -6.1059 0>}]
 dyn_skirt_mid_01_r [{shapeType=Capsule, length=14.75, radius=7.75, shapeName=dyn_skirt_mid_01_r_capsule, center=<5.6 -5.5059 0>}]
 dyn_skirt_mid_02_l [{shapeType=Capsule, length=14.75, radius=7.75, shapeName=dyn_skirt_mid_02_l_capsule, center=<5.6 -5.6 0>}]
 dyn_skirt_mid_02_r [{shapeType=Capsule, length=14.75, radius=7.75, shapeName=dyn_skirt_mid_02_r_capsule, center=<6 -5.5 0>}]
 dyn_skirt_mid_03_l [{shapeType=Capsule, length=14.75, radius=7.75, shapeName=dyn_skirt_mid_03_l_capsule, center=<6.6 -5.5 0>}]
 dyn_skirt_mid_03_r [{shapeType=Capsule, length=24.5, radius=7.75, shapeName=dyn_skirt_mid_03_r_capsule, center=<8 -5 0>}]
 dyn_skirt_mid_04_l [{shapeType=Capsule, length=18.25, radius=9.25, shapeName=dyn_skirt_mid_04_l_capsule, center=<8 -3.3 0>}]
 dyn_skirt_mid_04_r [{shapeType=Capsule, length=19, radius=9.25, shapeName=dyn_skirt_mid_04_r_capsule, center=<8 -4.5 0>}]
 dyn_skirt_side_01_l [{shapeType=Capsule, length=14.75, radius=7.75, shapeName=dyn_skirt_side_01_l_capsule, center=<5.6 -4.5 0>}]
 dyn_skirt_side_01_r [{shapeType=Capsule, length=14.75, radius=7.75, shapeName=dyn_skirt_side_01_r_capsule, center=<5.6 -5.5277 0>}]
 dyn_skirt_side_02_l [{shapeType=Capsule, length=15.25, radius=7.25, shapeName=dyn_skirt_side_02_l_capsule, center=<5.6 -2.9947 0>}]
 dyn_skirt_side_02_r [{shapeType=Capsule, length=14.75, radius=7.75, shapeName=dyn_skirt_side_02_r_capsule, center=<5.6 -7.8991 0>}]
 dyn_skirt_side_03_l [{shapeType=Capsule, length=15.25, radius=7.25, shapeName=dyn_skirt_side_03_l_capsule, center=<5.6 -2.6579 0>}]
 dyn_skirt_side_03_r [{shapeType=Capsule, length=14.75, radius=7.75, shapeName=dyn_skirt_side_03_r_capsule, center=<6.6 -8.1676 0>}]
 dyn_skirt_side_04_l [{shapeType=Capsule, length=18.75, radius=8.25, shapeName=dyn_skirt_side_04_l_capsule, center=<7 -3.2 0>}]
 dyn_skirt_side_04_r [{shapeType=Capsule, length=24.75, radius=8.5, shapeName=dyn_skirt_side_04_r_capsule, center=<8 -7.0098 0>}]
 pelvis [{shapeType=Capsule, length=18.174, radius=11.7894, center=<-1.1356 -0.0002 1.018>, rotation=<rot 90 -160.5288 -160.5287>}]
 root [{shapeType=Capsule, length=4, radius=31, shapeName=root_capsule, rotation=<rot 0 90 90>}]
 thigh_l [{shapeType=Capsule, length=21.25, radius=10.25, center=<-24.8764 -2.8482 0.5566>, rotation=<rot 87.6116 -179.9999 -179.9999>}]
 thigh_r [{shapeType=Capsule, length=23.75, radius=9.5, center=<25.5856 -0.5432 0.9419>, rotation=<rot 87.4684 180 0>}]
 thigh_twist_01_l [{shapeType=Capsule, length=18.0293, radius=9.4249, shapeName=thigh_twist_01_l_capsule, center=<7.2699 1.1957 0.874>, rotation=<rot 90 0 0>}]
 thigh_twist_01_r [{
   shapeType=Capsule
   length=14.9297
   radius=8.6924
   shapeName=thigh_twist_01_r_capsule
   center=<-11.9183 -5.9698 -0.8378>
   rotation=<rot 90 0 0>
  }]
}
constraints [
 {
  bone1Name=thigh_l
  bone2Name=pelvis
  swing1LimitDegrees=45
  swing1Motion=Free
  swing2LimitDegrees=45
  swing2Motion=Free
  twistLimitDegrees=45
  twistMotion=Free
 }
 {bone1Name=dyn_skirt_back_04_l, bone2Name=dyn_skirt_back_03_l, swing1LimitDegrees=70, swing2LimitDegrees=30, twistLimitDegrees=5}
 {bone1Name=dyn_skirt_front_01_r, bone2Name=pelvis, swing1LimitDegrees=45, swing2LimitDegrees=45, twistLimitDegrees=5}
 {bone1Name=dyn_skirt_front_02_r, bone2Name=dyn_skirt_front_01_r, swing1LimitDegrees=50, swing2LimitDegrees=30, twistLimitDegrees=2}
 {bone1Name=dyn_skirt_front_03_r, bone2Name=dyn_skirt_front_02_r, swing1LimitDegrees=50, swing2LimitDegrees=20, twistLimitDegrees=2}
 {bone1Name=dyn_skirt_front_04_r, bone2Name=dyn_skirt_front_03_r, swing1LimitDegrees=50, swing2LimitDegrees=20, twistLimitDegrees=2}
 {bone1Name=dyn_skirt_side_01_r, bone2Name=pelvis, swing1LimitDegrees=50, swing2LimitDegrees=20, twistLimitDegrees=2}
 {bone1Name=dyn_skirt_side_02_r, bone2Name=dyn_skirt_side_01_r, swing1LimitDegrees=50, swing2LimitDegrees=20, twistLimitDegrees=2}
 {bone1Name=dyn_skirt_side_03_r, bone2Name=dyn_skirt_side_02_r, swing1LimitDegrees=50, swing2LimitDegrees=20, twistLimitDegrees=2}
 {bone1Name=dyn_skirt_side_04_r, bone2Name=dyn_skirt_side_03_r, swing1LimitDegrees=60, swing2LimitDegrees=20, twistLimitDegrees=2}
 {bone1Name=dyn_skirt_mid_01_r, bone2Name=pelvis, swing1LimitDegrees=50, swing2LimitDegrees=20, twistLimitDegrees=2}
 {bone1Name=dyn_skirt_mid_02_r, bone2Name=dyn_skirt_mid_01_r, swing1LimitDegrees=50, swing2LimitDegrees=20, twistLimitDegrees=2}
 {bone1Name=dyn_skirt_mid_03_r, bone2Name=dyn_skirt_mid_02_r, swing1LimitDegrees=50, swing2LimitDegrees=20, twistLimitDegrees=2}
 {bone1Name=dyn_skirt_mid_04_r, bone2Name=dyn_skirt_mid_03_r, swing1LimitDegrees=50, swing2LimitDegrees=20, twistLimitDegrees=2}
 {bone1Name=dyn_skirt_back_01_r, bone2Name=pelvis, swing1LimitDegrees=50, swing2LimitDegrees=30, twistLimitDegrees=5}
 {bone1Name=dyn_skirt_back_02_r, bone2Name=dyn_skirt_back_01_r, swing1LimitDegrees=70, swing2LimitDegrees=30, twistLimitDegrees=5}
 {bone1Name=dyn_skirt_back_03_r, bone2Name=dyn_skirt_back_02_r, swing1LimitDegrees=70, swing2LimitDegrees=30, twistLimitDegrees=5}
 {bone1Name=dyn_skirt_back_04_r, bone2Name=dyn_skirt_back_03_r, swing1LimitDegrees=70, swing2LimitDegrees=30, twistLimitDegrees=5}
 {bone1Name=dyn_skirt_front_01_l, bone2Name=pelvis, swing1LimitDegrees=45, swing2LimitDegrees=45, twistLimitDegrees=20}
 {bone1Name=dyn_skirt_front_02_l, bone2Name=dyn_skirt_front_01_l, swing1LimitDegrees=50, swing2LimitDegrees=20, twistLimitDegrees=2}
 {bone1Name=dyn_skirt_front_03_l, bone2Name=dyn_skirt_front_02_l, swing1LimitDegrees=50, swing2LimitDegrees=20, twistLimitDegrees=2}
 {bone1Name=dyn_skirt_front_04_l, bone2Name=dyn_skirt_front_03_l, swing1LimitDegrees=50, swing2LimitDegrees=20, twistLimitDegrees=2}
 {bone1Name=dyn_skirt_side_01_l, bone2Name=pelvis, swing1LimitDegrees=50, swing2LimitDegrees=20, twistLimitDegrees=2}
 {bone1Name=dyn_skirt_side_02_l, bone2Name=dyn_skirt_side_01_l, swing1LimitDegrees=50, swing2LimitDegrees=20, twistLimitDegrees=2}
 {bone1Name=dyn_skirt_side_03_l, bone2Name=dyn_skirt_side_02_l, swing1LimitDegrees=50, swing2LimitDegrees=20, twistLimitDegrees=2}
 {bone1Name=dyn_skirt_side_04_l, bone2Name=dyn_skirt_side_03_l, swing1LimitDegrees=50, swing2LimitDegrees=20, twistLimitDegrees=2}
 {bone1Name=dyn_skirt_mid_01_l, bone2Name=pelvis, swing1LimitDegrees=50, swing2LimitDegrees=20, twistLimitDegrees=2}
 {bone1Name=dyn_skirt_mid_02_l, bone2Name=dyn_skirt_mid_01_l, swing1LimitDegrees=50, swing2LimitDegrees=20, twistLimitDegrees=2}
 {bone1Name=dyn_skirt_mid_03_l, bone2Name=dyn_skirt_mid_02_l, swing1LimitDegrees=50, swing2LimitDegrees=20, twistLimitDegrees=2}
 {bone1Name=dyn_skirt_mid_04_l, bone2Name=dyn_skirt_mid_03_l, swing1LimitDegrees=50, swing2LimitDegrees=20, twistLimitDegrees=2}
 {bone1Name=dyn_skirt_back_01_l, bone2Name=pelvis, swing1LimitDegrees=50, swing2LimitDegrees=30, twistLimitDegrees=5}
 {bone1Name=dyn_skirt_back_02_l, bone2Name=dyn_skirt_back_01_l, swing1LimitDegrees=70, swing2LimitDegrees=30, twistLimitDegrees=5}
 {bone1Name=dyn_skirt_back_03_l, bone2Name=dyn_skirt_back_02_l, swing1LimitDegrees=70, swing2LimitDegrees=30, twistLimitDegrees=5}
 {
  bone1Name=thigh_r
  bone2Name=pelvis
  swing1LimitDegrees=45
  swing1Motion=Free
  swing2LimitDegrees=45
  swing2Motion=Free
  twistLimitDegrees=45
  twistMotion=Free
 }
 {
  bone1Name=calf_l
  bone2Name=thigh_l
  swing1LimitDegrees=45
  swing1Motion=Free
  swing2LimitDegrees=45
  swing2Motion=Free
  twistLimitDegrees=45
  twistMotion=Free
 }
 {
  bone1Name=calf_r
  bone2Name=thigh_r
  swing1LimitDegrees=45
  swing1Motion=Free
  swing2LimitDegrees=45
  swing2Motion=Free
  twistLimitDegrees=45
  twistMotion=Free
 }
 {bone1Name=dyn_bag, bone2Name=pelvis, swing1LimitDegrees=20, swing2LimitDegrees=60, twistLimitDegrees=5}
 {
  bone1Name=thigh_twist_01_l
  bone2Name=thigh_l
  swing1LimitDegrees=45
  swing1Motion=Free
  swing2LimitDegrees=45
  swing2Motion=Free
  twistLimitDegrees=45
  twistMotion=Free
 }
 {
  bone1Name=thigh_twist_01_r
  bone2Name=thigh_r
  swing1LimitDegrees=45
  swing1Motion=Free
  swing2LimitDegrees=45
  swing2Motion=Free
  twistLimitDegrees=45
  twistMotion=Free
 }
 {
  bone1Name=pelvis
  bone2Name=root
  swing1LimitDegrees=45
  swing1Motion=Free
  swing2LimitDegrees=45
  swing2Motion=Free
  twistLimitDegrees=45
  twistMotion=Free
 }
]
