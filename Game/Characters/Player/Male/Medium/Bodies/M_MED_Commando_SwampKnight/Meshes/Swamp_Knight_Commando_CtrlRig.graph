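# Control Rig for the Swamp Knight Commando body that applies corrective offsets to its left shoulder pad and coat skirt bones.
# Spherical pose readers on the left clavicle and upper arm weight shoulder pad offsets, with separate sets for normal, skydiving and skydive-float states.
# Two further sequence steps pose left and right skirt bones for knee and baseball slides, each weighted by its slide variable smoothed through AlphaInterp.
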ControlRigBlueprint /Game/Characters/Player/Male/Medium/Bodies/M_MED_Commando_SwampKnight/Meshes/Swamp_Knight_Commando_CtrlRig
graphs [
 ContainedGraph {nodes [
   n1 = RigVMFunctionEntryNode()
   SequenceAggregate = &RigVMFunction_Sequence(ExecuteContext=@n1.ExecuteContext)
   SequenceAggregate_1 = &RigVMFunction_Sequence(ExecuteContext=@SequenceAggregate.B)
   RigVMFunctionReturnNode(A=@SequenceAggregate.A, B=@SequenceAggregate_1.A, C=@SequenceAggregate_1.B)
  ]}
 ContainedGraph {nodes [
   RigVMFunctionEntryNode()
   > RigVMFunctionReturnNode()
   &RigVMFunction_MathFloatAdd(A=0, B=0) {template="Add::Execute(in A,in B,out Result)"}
   &RigVMFunction_MathFloatAdd(A=0, B=0) {template="Add::Execute(in A,in B,out Result)"}
   &RigVMFunction_MathFloatAdd(A=0, B=0) {template="Add::Execute(in A,in B,out Result)"}
   &RigVMFunction_MathFloatAdd(A=0, B=0) {template="Add::Execute(in A,in B,out Result)"}
   &RigVMFunction_MathFloatAdd(A=0, B=0) {template="Add::Execute(in A,in B,out Result)"}
   &RigVMFunction_MathFloatAdd(A=0, B=0) {template="Add::Execute(in A,in B,out Result)"}
   &RigVMFunction_MathFloatAdd(A=0, B=0) {template="Add::Execute(in A,in B,out Result)"}
   &RigVMFunction_MathFloatAdd(A=0, B=0) {template="Add::Execute(in A,in B,out Result)"}
   &RigVMFunction_MathFloatAdd(A=0, B=0) {template="Add::Execute(in A,in B,out Result)"}
   &RigVMFunction_MathQuaternionFromEuler(Euler=(X=0,Y=0,Z=0))
   &RigVMFunction_MathQuaternionToEuler(RotationOrder=ZYX, Value=(X=0,Y=0,Z=0,W=1))
   RigUnit_GetTransform = &RigUnit_GetTransform(Item=(Type=Bone,Name="dyn_lapel_high_l"), Space=LocalSpace, bInitial=false)
   RigUnit_SetTransform = &RigUnit_SetTransform(
    ExecutePin=()
    Item=(Type=Bone,Name="dyn_lapel_high_l")
    Space=LocalSpace
    Value=(Rotation=(X=0.0132,Y=0.4099,Z=-0.0222,W=0.9118),Translation=(X=23.6045,Y=-6.37,Z=140.1115),Scale3D=(X=1,Y=1,Z=1))
    Weight=1
    bInitial=False
   )
  ]}
 RigVMFunctionLibrary {nodes [Set Transforms for Bone = RigVMCollapseNode(
    ExecuteContext=()
    Joint=()
    Rotate X=0
    Rotate Y=0
    Rotate Z=0
    Scale X=0
    Scale Y=0
    Scale Z=0
    Translate X=0
    Translate Y=0
    Translate Z=0
    Weight=1
   )]}
 RigVMModel {nodes [
   RigUnit_BeginExecution = &RigUnit_BeginExecution()
   SequenceAggregate = RigVMAggregateNode(ExecuteContext=@RigUnit_BeginExecution.ExecutePin)
   Branch_RigVMUnitNode = &RigVMFunction_ControlFlowBranch(ExecuteContext=@SequenceAggregate.A, Condition=RigVMVariableNode() {variable=Is Sky Diving_1}.Value)
   n1 = &RigUnit_SphericalPoseReader(
    ActiveRegionScaleFactors=(PositiveWidth=1,NegativeWidth=1,PositiveHeight=1,NegativeHeight=1)
    ActiveRegionSize=0.03
    Debug=(bDrawDebug=True,DebugScale=25,DebugSegments=20,DebugThickness=0.25)
    DriverAxis=(X=1,Y=0,Z=0)
    DriverItem=(Type=Bone,Name="clavicle_l")
    FalloffRegionScaleFactors=(PositiveWidth=1,NegativeWidth=1,PositiveHeight=1,NegativeHeight=1)
    FalloffSize=0.07
    OptionalParentItem=(Type=Bone)
    RotationOffset=(X=-10,Y=-90,Z=0)
    ExecutePin=@Branch_RigVMUnitNode.False
   )
   > &RigUnit_ModifyTransforms(
    ItemToModify=((Item=(Type=Bone,Name="dyn_shoulder_pad_00_b"),Transform=(Rotation=(X=0,Y=0,Z=0,W=1),Translation=(X=0,Y=0,Z=0),Scale3D=(X=1,Y=1,Z=1))))
    ItemToModify.0.Transform.Rotation=&RigVMFunction_MathQuaternionFromEuler(Euler=(X=0,Y=0,Z=-11)).Result
    Weight=@n1.OutputParam
   )
   > n2 = &RigUnit_SphericalPoseReader(
    ActiveRegionScaleFactors=(PositiveWidth=1,NegativeWidth=1,PositiveHeight=1,NegativeHeight=1)
    ActiveRegionSize=0.05
    Debug=(bDrawDebug=True,DebugScale=25,DebugSegments=20,DebugThickness=0.25)
    DriverAxis=(X=1,Y=0,Z=0)
    DriverItem=(Type=Bone,Name="clavicle_l")
    FalloffRegionScaleFactors=(PositiveWidth=1,NegativeWidth=1,PositiveHeight=1,NegativeHeight=1)
    FalloffSize=0.1
    OptionalParentItem=(Type=Bone)
    RotationOffset=(X=5,Y=-70,Z=0)
   )
   > &RigUnit_ModifyTransforms(
    ItemToModify=((Item=(Type=Bone,Name="dyn_shoulder_pad_00_b"),Transform=(Rotation=(X=0,Y=0,Z=0,W=1),Translation=(X=0,Y=1,Z=0),Scale3D=(X=1,Y=1,Z=1))))
    ItemToModify.0.Transform.Rotation=&RigVMFunction_MathQuaternionFromEuler(Euler=(X=0,Y=0,Z=13)).Result
    Weight=@n2.OutputParam
   )
   > n3 = &RigUnit_SphericalPoseReader(
    ActiveRegionScaleFactors=(PositiveWidth=1,NegativeWidth=1,PositiveHeight=1,NegativeHeight=1)
    ActiveRegionSize=0.1
    Debug=(bDrawDebug=True,DebugScale=25,DebugSegments=20,DebugThickness=0.25)
    DriverAxis=(X=1,Y=0,Z=0)
    DriverItem=(Type=Bone,Name="clavicle_l")
    FalloffRegionScaleFactors=(PositiveWidth=1,NegativeWidth=1,PositiveHeight=1,NegativeHeight=1)
    FalloffSize=0.2
    OptionalParentItem=(Type=Bone)
    RotationOffset=(X=5,Y=-20,Z=0)
   )
   > &RigUnit_ModifyTransforms(
    ItemToModify=((Item=(Type=Bone,Name="dyn_shoulder_pad_00_b"),Transform=(Rotation=(X=0,Y=0,Z=0,W=1),Translation=(X=0,Y=0,Z=0),Scale3D=(X=1,Y=1,Z=1))))
    ItemToModify.0.Transform.Rotation=&RigVMFunction_MathQuaternionFromEuler(Euler=(X=10,Y=12,Z=-10)).Result
    Weight=@n3.OutputParam
   )
   > n4 = &RigUnit_SphericalPoseReader(
    ActiveRegionScaleFactors=(PositiveWidth=1,NegativeWidth=1,PositiveHeight=1,NegativeHeight=1)
    ActiveRegionSize=0.02
    Debug=(bDrawDebug=True,DebugScale=25,DebugSegments=20,DebugThickness=0.25)
    DriverAxis=(X=1,Y=0,Z=0)
    DriverItem=(Type=Bone,Name="clavicle_l")
    FalloffRegionScaleFactors=(PositiveWidth=1,NegativeWidth=1,PositiveHeight=1,NegativeHeight=1)
    FalloffSize=0.05
    OptionalParentItem=(Type=Bone)
    RotationOffset=(X=-20,Y=-50,Z=0)
   )
   > &RigUnit_ModifyTransforms(
    ItemToModify=((Item=(Type=Bone,Name="dyn_shoulder_pad_00_b"),Transform=(Rotation=(X=0,Y=0,Z=0,W=1),Translation=(X=-3,Y=-2,Z=0),Scale3D=(X=1,Y=1,Z=1))))
    ItemToModify.0.Transform.Rotation=&RigVMFunction_MathQuaternionFromEuler(Euler=(X=0,Y=0,Z=-15)).Result
    Weight=@n4.OutputParam
   )
   > n5 = &RigUnit_SphericalPoseReader(
    ActiveRegionScaleFactors=(PositiveWidth=2.2,NegativeWidth=1,PositiveHeight=1,NegativeHeight=2.5)
    ActiveRegionSize=0.05
    Debug=(bDrawDebug=True,DebugScale=25,DebugSegments=20,DebugThickness=0.25)
    DriverAxis=(X=1,Y=0,Z=0)
    DriverItem=(Type=Bone,Name="upperarm_l")
    FalloffRegionScaleFactors=(PositiveWidth=1,NegativeWidth=2,PositiveHeight=1,NegativeHeight=1)
    FalloffSize=0.18
    OptionalParentItem=(Type=Bone)
    RotationOffset=(X=60,Y=15,Z=-60)
   )
   > &RigUnit_ModifyTransforms(
    ItemToModify=((Item=(Type=Bone,Name="dyn_shoulder_pad_00_b"),Transform=(Rotation=(X=0,Y=0,Z=0,W=1),Translation=(X=6,Y=-1,Z=-2),Scale3D=(X=1,Y=1,Z=1))))
    ItemToModify.0.Transform.Rotation=&RigVMFunction_MathQuaternionFromEuler(Euler=(X=0,Y=0,Z=10)).Result
    Weight=@n5.OutputParam
   )
   > n6 = &RigUnit_SphericalPoseReader(
    ActiveRegionScaleFactors=(PositiveWidth=1,NegativeWidth=1,PositiveHeight=1,NegativeHeight=1)
    ActiveRegionSize=0.035
    Debug=(bDrawDebug=True,DebugScale=25,DebugSegments=20,DebugThickness=0.25)
    DriverAxis=(X=1,Y=0,Z=0)
    DriverItem=(Type=Bone,Name="upperarm_l")
    FalloffRegionScaleFactors=(PositiveWidth=1,NegativeWidth=1,PositiveHeight=0.5,NegativeHeight=1)
    FalloffSize=0.1
    OptionalParentItem=(Type=Bone)
    RotationOffset=(X=32,Y=-10,Z=-65)
   )
   > &RigUnit_ModifyTransforms(
    ItemToModify=((Item=(Type=Bone,Name="dyn_shoulder_pad_00_b"),Transform=(Rotation=(X=0,Y=0,Z=0,W=1),Translation=(X=3,Y=-5,Z=0),Scale3D=(X=1,Y=1,Z=1))))
    ItemToModify.0.Transform.Rotation=&RigVMFunction_MathQuaternionFromEuler(Euler=(X=0,Y=0,Z=-20)).Result
    Weight=@n6.OutputParam
   )
   > n7 = &RigUnit_SphericalPoseReader(
    ActiveRegionScaleFactors=(PositiveWidth=0.7,NegativeWidth=0.7,PositiveHeight=0.7,NegativeHeight=0.7)
    ActiveRegionSize=0.2
    Debug=(bDrawDebug=True,DebugScale=25,DebugSegments=20,DebugThickness=0.25)
    DriverAxis=(X=1,Y=0,Z=0)
    DriverItem=(Type=Bone,Name="upperarm_l")
    FalloffRegionScaleFactors=(PositiveWidth=1,NegativeWidth=1,PositiveHeight=1,NegativeHeight=1)
    FalloffSize=0.1
    OptionalParentItem=(Type=Bone)
    RotationOffset=(X=140,Y=40,Z=0)
   )
   > &RigUnit_ModifyTransforms(
    ItemToModify=((Item=(Type=Bone,Name="dyn_shoulder_pad_00_b"),Transform=(Rotation=(X=0,Y=0,Z=0,W=1),Translation=(X=2,Y=1,Z=0),Scale3D=(X=1,Y=1,Z=1))))
    ItemToModify.0.Transform.Rotation=&RigVMFunction_MathQuaternionFromEuler(Euler=(X=-10,Y=-5,Z=-2)).Result
    Weight=@n7.OutputParam
   )
   Branch_1_RigVMUnitNode = &RigVMFunction_ControlFlowBranch(ExecuteContext=@Branch_RigVMUnitNode.True, Condition=RigVMVariableNode() {variable=SkydiveFloat_0}.Value)
   n8 = &RigUnit_SphericalPoseReader(
    ActiveRegionScaleFactors=(PositiveWidth=1,NegativeWidth=1,PositiveHeight=1,NegativeHeight=1)
    ActiveRegionSize=0.3
    Debug=(bDrawDebug=True,DebugScale=25,DebugSegments=20,DebugThickness=0.25)
    DriverAxis=(X=1,Y=0,Z=0)
    DriverItem=(Type=Bone,Name="upperarm_l")
    FalloffRegionScaleFactors=(PositiveWidth=1,NegativeWidth=1,PositiveHeight=1,NegativeHeight=1)
    FalloffSize=0.4
    OptionalParentItem=(Type=Bone)
    RotationOffset=(X=0,Y=-60,Z=0)
    ExecutePin=@Branch_1_RigVMUnitNode.False
   )
   > &RigUnit_ModifyTransforms(
    ItemToModify=((Item=(Type=Bone,Name="dyn_shoulder_pad_00_b"),Transform=(Rotation=(X=0,Y=0,Z=0,W=1),Translation=(X=6,Y=-2,Z=2),Scale3D=(X=1,Y=1,Z=1))))
    ItemToModify.0.Transform.Rotation=&RigVMFunction_MathQuaternionFromEuler(Euler=(X=10,Y=30,Z=10)).Result
    Weight=@n8.OutputParam
   )
   > n9 = &RigUnit_SphericalPoseReader(
    ActiveRegionScaleFactors=(PositiveWidth=1,NegativeWidth=1,PositiveHeight=1,NegativeHeight=1)
    ActiveRegionSize=0.05
    Debug=(bDrawDebug=True,DebugScale=25,DebugSegments=20,DebugThickness=0.25)
    DriverAxis=(X=1,Y=0,Z=0)
    DriverItem=(Type=Bone,Name="upperarm_l")
    FalloffRegionScaleFactors=(PositiveWidth=1,NegativeWidth=1,PositiveHeight=1,NegativeHeight=1)
    FalloffSize=0.08
    OptionalParentItem=(Type=Bone)
    RotationOffset=(X=-10,Y=-65,Z=-30)
   )
   > &RigUnit_ModifyTransforms(
    ItemToModify=((Item=(Type=Bone,Name="dyn_shoulder_pad_00_b"),Transform=(Rotation=(X=0,Y=0,Z=0,W=1),Translation=(X=5,Y=0,Z=1),Scale3D=(X=1,Y=1,Z=1))))
    ItemToModify.0.Transform.Rotation=&RigVMFunction_MathQuaternionFromEuler(Euler=(X=0,Y=0,Z=0)).Result
    Weight=@n9.OutputParam
   )
   > n10 = &RigUnit_SphericalPoseReader(
    ActiveRegionScaleFactors=(PositiveWidth=1,NegativeWidth=1,PositiveHeight=1,NegativeHeight=1)
    ActiveRegionSize=0.03
    Debug=(bDrawDebug=True,DebugScale=25,DebugSegments=20,DebugThickness=0.25)
    DriverAxis=(X=1,Y=0,Z=0)
    DriverItem=(Type=Bone,Name="upperarm_l")
    FalloffRegionScaleFactors=(PositiveWidth=2,NegativeWidth=1,PositiveHeight=1,NegativeHeight=1)
    FalloffSize=0.05
    OptionalParentItem=(Type=Bone)
    RotationOffset=(X=-20,Y=-30,Z=0)
   )
   > &RigUnit_ModifyTransforms(
    ItemToModify=((Item=(Type=Bone,Name="dyn_shoulder_pad_00_b"),Transform=(Rotation=(X=0,Y=0,Z=0,W=1),Translation=(X=1,Y=0,Z=1),Scale3D=(X=1,Y=1,Z=1))))
    ItemToModify.0.Transform.Rotation=&RigVMFunction_MathQuaternionFromEuler(Euler=(X=0,Y=0,Z=0)).Result
    Weight=@n10.OutputParam
   )
   n11 = &RigUnit_SphericalPoseReader(
    ActiveRegionScaleFactors=(PositiveWidth=1,NegativeWidth=1,PositiveHeight=1,NegativeHeight=1)
    ActiveRegionSize=0.3
    Debug=(bDrawDebug=True,DebugScale=25,DebugSegments=20,DebugThickness=0.25)
    DriverAxis=(X=1,Y=0,Z=0)
    DriverItem=(Type=Bone,Name="upperarm_l")
    FalloffRegionScaleFactors=(PositiveWidth=1,NegativeWidth=1,PositiveHeight=1,NegativeHeight=1)
    FalloffSize=0.4
    OptionalParentItem=(Type=Bone)
    RotationOffset=(X=0,Y=-60,Z=0)
    ExecutePin=@Branch_1_RigVMUnitNode.True
   )
   > &RigUnit_ModifyTransforms(
    ItemToModify=((Item=(Type=Bone,Name="dyn_shoulder_pad_00_b"),Transform=(Rotation=(X=0,Y=0,Z=0,W=1),Translation=(X=5,Y=2,Z=-0.5),Scale3D=(X=1,Y=1,Z=1))))
    ItemToModify.0.Transform.Rotation=&RigVMFunction_MathQuaternionFromEuler(Euler=(X=-15,Y=20,Z=20)).Result
    Weight=@n11.OutputParam
   )
   AlphaInterp = &RigVMFunction_AlphaInterp(InterpSpeedDecreasing=7, InterpSpeedIncreasing=5, Scale=1, bClampResult=False, Value=RigVMVariableNode() {variable=Knee_Slide}.Value)
   &RigUnit_ModifyTransforms(
    ItemToModify=((Item=(Type=Bone,Name="dyn_skirt_front_01_l"),Transform=(Rotation=(X=0,Y=0,Z=0,W=1),Translation=(X=0,Y=0,Z=0),Scale3D=(X=1,Y=1,Z=1))))
    Weight=@AlphaInterp.Result
    ItemToModify.0.Transform.Rotation=&RigVMFunction_MathQuaternionFromEuler(Euler=(X=0,Y=0,Z=0)).Result
    ExecutePin=@SequenceAggregate.B
   )
   > &RigUnit_ModifyTransforms(
    ItemToModify=((Item=(Type=Bone,Name="dyn_skirt_side_01_l"),Transform=(Rotation=(X=0,Y=0,Z=0,W=1),Translation=(X=0,Y=0,Z=0),Scale3D=(X=1,Y=1,Z=1))))
    Weight=@AlphaInterp.Result
    ItemToModify.0.Transform.Rotation=&RigVMFunction_MathQuaternionFromEuler(Euler=(X=0,Y=0,Z=-20)).Result
   )
   > &RigUnit_ModifyTransforms(
    ItemToModify=((Item=(Type=Bone,Name="dyn_skirt_back_01_l"),Transform=(Rotation=(X=0,Y=0,Z=0,W=1),Translation=(X=0,Y=0,Z=0),Scale3D=(X=1,Y=1,Z=1))))
    Weight=@AlphaInterp.Result
    ItemToModify.0.Transform.Rotation=&RigVMFunction_MathQuaternionFromEuler(Euler=(X=0,Y=0,Z=-60)).Result
   )
   > &RigUnit_ModifyTransforms(
    ItemToModify=((Item=(Type=Bone,Name="dyn_skirt_mid_01_l"),Transform=(Rotation=(X=0,Y=0,Z=0,W=1),Translation=(X=0,Y=0,Z=0),Scale3D=(X=1,Y=1,Z=1))))
    Weight=@AlphaInterp.Result
    ItemToModify.0.Transform.Rotation=&RigVMFunction_MathQuaternionFromEuler(Euler=(X=0,Y=-20,Z=-60)).Result
   )
   > &RigUnit_ModifyTransforms(
    ItemToModify=((Item=(Type=Bone,Name="dyn_skirt_back_04_l"),Transform=(Rotation=(X=0,Y=0,Z=0,W=1),Translation=(X=0,Y=0,Z=0),Scale3D=(X=1,Y=1,Z=1))))
    Weight=@AlphaInterp.Result
    ItemToModify.0.Transform.Rotation=&RigVMFunction_MathQuaternionFromEuler(Euler=(X=0,Y=20,Z=20)).Result
   )
   > &RigUnit_ModifyTransforms(
    ItemToModify=((Item=(Type=Bone,Name="dyn_skirt_front_01_r"),Transform=(Rotation=(X=0,Y=0,Z=0,W=1),Translation=(X=0,Y=0,Z=0),Scale3D=(X=1,Y=1,Z=1))))
    Weight=@AlphaInterp.Result
    ItemToModify.0.Transform.Rotation=&RigVMFunction_MathQuaternionFromEuler(Euler=(X=0,Y=-30,Z=0)).Result
   )
   > &RigUnit_ModifyTransforms(
    ItemToModify=((Item=(Type=Bone,Name="dyn_skirt_side_01_r"),Transform=(Rotation=(X=0,Y=0,Z=0,W=1),Translation=(X=0,Y=0,Z=0),Scale3D=(X=1,Y=1,Z=1))))
    Weight=@AlphaInterp.Result
    ItemToModify.0.Transform.Rotation=&RigVMFunction_MathQuaternionFromEuler(Euler=(X=0,Y=-20,Z=-50)).Result
   )
   > &RigUnit_ModifyTransforms(
    ItemToModify=((Item=(Type=Bone,Name="dyn_skirt_back_01_r"),Transform=(Rotation=(X=0,Y=0,Z=0,W=1),Translation=(X=-20,Y=0,Z=0),Scale3D=(X=1,Y=1,Z=1))))
    Weight=@AlphaInterp.Result
    ItemToModify.0.Transform.Rotation=&RigVMFunction_MathQuaternionFromEuler(Euler=(X=0,Y=0,Z=-40)).Result
   )
   > &RigUnit_ModifyTransforms(
    ItemToModify=((Item=(Type=Bone,Name="dyn_skirt_mid_01_r"),Transform=(Rotation=(X=0,Y=0,Z=0,W=1),Translation=(X=0,Y=0,Z=0),Scale3D=(X=1,Y=1,Z=1))))
    Weight=@AlphaInterp.Result
    ItemToModify.0.Transform.Rotation=&RigVMFunction_MathQuaternionFromEuler(Euler=(X=0,Y=-10,Z=-50)).Result
   )
   > &RigUnit_ModifyTransforms(
    ItemToModify=((Item=(Type=Bone,Name="dyn_skirt_back_04_r"),Transform=(Rotation=(X=0,Y=0,Z=0,W=1),Translation=(X=0,Y=0,Z=0),Scale3D=(X=1,Y=1,Z=1))))
    Weight=@AlphaInterp.Result
    ItemToModify.0.Transform.Rotation=&RigVMFunction_MathQuaternionFromEuler(Euler=(X=0,Y=0,Z=50)).Result
   )
   n12 = &RigVMFunction_AlphaInterp(InterpSpeedDecreasing=7, InterpSpeedIncreasing=5, Scale=1, bClampResult=False, Value=RigVMVariableNode() {variable=BaseBall_Slide}.Value)
   &RigUnit_ModifyTransforms(
    ItemToModify=((Item=(Type=Bone,Name="dyn_skirt_front_01_l"),Transform=(Rotation=(X=0,Y=0,Z=0,W=1),Translation=(X=0,Y=0,Z=0),Scale3D=(X=1,Y=1,Z=1))))
    Weight=@n12.Result
    ItemToModify.0.Transform.Rotation=&RigVMFunction_MathQuaternionFromEuler(Euler=(X=0,Y=0,Z=0)).Result
    ExecutePin=@SequenceAggregate.C
   )
   > &RigUnit_ModifyTransforms(
    ItemToModify=((Item=(Type=Bone,Name="dyn_skirt_side_01_l"),Transform=(Rotation=(X=0,Y=0,Z=0,W=1),Translation=(X=0,Y=0,Z=0),Scale3D=(X=1,Y=1,Z=1))))
    Weight=@n12.Result
    ItemToModify.0.Transform.Rotation=&RigVMFunction_MathQuaternionFromEuler(Euler=(X=0,Y=0,Z=-70)).Result
   )
   > &RigUnit_ModifyTransforms(
    ItemToModify=((Item=(Type=Bone,Name="dyn_skirt_back_01_l"),Transform=(Rotation=(X=0,Y=0,Z=0,W=1),Translation=(X=0,Y=0,Z=0),Scale3D=(X=1,Y=1,Z=1))))
    Weight=@n12.Result
    ItemToModify.0.Transform.Rotation=&RigVMFunction_MathQuaternionFromEuler(Euler=(X=0,Y=0,Z=-80)).Result
   )
   > &RigUnit_ModifyTransforms(
    ItemToModify=((Item=(Type=Bone,Name="dyn_skirt_mid_01_l"),Transform=(Rotation=(X=0,Y=0,Z=0,W=1),Translation=(X=0,Y=0,Z=0),Scale3D=(X=1,Y=1,Z=1))))
    Weight=@n12.Result
    ItemToModify.0.Transform.Rotation=&RigVMFunction_MathQuaternionFromEuler(Euler=(X=0,Y=0,Z=-70)).Result
   )
   > &RigUnit_ModifyTransforms(
    ItemToModify=((Item=(Type=Bone,Name="dyn_skirt_front_01_r"),Transform=(Rotation=(X=0,Y=0,Z=0,W=1),Translation=(X=4,Y=5,Z=-5),Scale3D=(X=1,Y=1,Z=1))))
    Weight=@n12.Result
    ItemToModify.0.Transform.Rotation=&RigVMFunction_MathQuaternionFromEuler(Euler=(X=-40,Y=20,Z=40)).Result
   )
   > &RigUnit_ModifyTransforms(
    ItemToModify=((Item=(Type=Bone,Name="dyn_skirt_side_01_r"),Transform=(Rotation=(X=0,Y=0,Z=0,W=1),Translation=(X=-10,Y=0,Z=0),Scale3D=(X=1,Y=1,Z=1))))
    Weight=@n12.Result
    ItemToModify.0.Transform.Rotation=&RigVMFunction_MathQuaternionFromEuler(Euler=(X=0,Y=0,Z=-100)).Result
   )
   > &RigUnit_ModifyTransforms(
    ItemToModify=((Item=(Type=Bone,Name="dyn_skirt_back_01_r"),Transform=(Rotation=(X=0,Y=0,Z=0,W=1),Translation=(X=-10,Y=0,Z=0),Scale3D=(X=1,Y=1,Z=1))))
    Weight=@n12.Result
    ItemToModify.0.Transform.Rotation=&RigVMFunction_MathQuaternionFromEuler(Euler=(X=0,Y=0,Z=-80)).Result
   )
   > &RigUnit_ModifyTransforms(
    ItemToModify=((Item=(Type=Bone,Name="dyn_skirt_mid_01_r"),Transform=(Rotation=(X=0,Y=0,Z=0,W=1),Translation=(X=-10,Y=0,Z=0),Scale3D=(X=1,Y=1,Z=1))))
    Weight=@n12.Result
    ItemToModify.0.Transform.Rotation=&RigVMFunction_MathQuaternionFromEuler(Euler=(X=0,Y=0,Z=-80)).Result
   )
  ]}
]
hierarchy [
 root {loc=<0 0 0>, rot=<0 0 0 1>}
 pelvis {parent=root, loc=<0 -0.7071 93.1191>, rot=<0.0127 0.707 -0.0127 -0.707>}
 spine_01 {parent=pelvis, loc=<6.7494 -0.6388 0>, rot=<0 0 0.034 -0.9994>}
 spine_02 {parent=spine_01, loc=<7.2518 -0.035 0>, rot=<0 0 0.0238 -0.9997>}
 spine_03 {parent=spine_02, loc=<6.0104 0 0>, rot=<0 0 0.0149 -0.9999>}
 spine_04 {parent=spine_03, loc=<8.1024 0 0>, rot=<0 0 -0.0132 -0.9999>}
 spine_05 {parent=spine_04, loc=<10.0146 0 0>, rot=<0 0 -0.0015 -1>}
 clavicle_l {parent=spine_05, loc=<4.3575 -0.1466 -2.3223>, rot=<0.0465 -0.7289 -0.009 -0.6829>}
 upperarm_l {parent=clavicle_l, loc=<15.5111 0 0>, rot=<0.0226 -0.3234 -0.0088 -0.9459>}
 lowerarm_l {parent=upperarm_l, loc=<26.6514 0 0>, rot=<0.0391 -0.0099 -0.2453 -0.9686>}
 hand_l {parent=lowerarm_l, loc=<22.3491 0 0>, rot=<-0.7081 -0.0507 0.0975 -0.6975>}
 index_metacarpal_l {parent=hand_l, loc=<1.6535 0.0183 -2.96>, rot=<0.0158 -0.0124 -0.0173 -0.9996>}
 index_01_l {parent=index_metacarpal_l, loc=<8.3008 0.0357 -0.0524>, rot=<-0.0044 0.0485 0.0924 -0.9945>}
 index_02_l {parent=index_01_l, loc=<3.3587 -0.3004 0.0452>, rot=<-0.002 0.0027 0.3225 -0.9466>}
 index_03_l {parent=index_02_l, loc=<2.5578 -0.123 -0.0252>, rot=<-0.002 0.0067 0.2558 -0.9667>}
 middle_metacarpal_l {parent=hand_l, loc=<2.1856 -0.1725 -0.5984>, rot=<0.006 0.0168 -0.0329 -0.9993>}
 middle_01_l {parent=middle_metacarpal_l, loc=<7.3275 -0.2447 0.0129>, rot=<-0.0208 0.0042 0.0858 -0.9961>}
 middle_02_l {parent=middle_01_l, loc=<4.0123 0.1401 0.1175>, rot=<-0.0218 0.0258 0.3512 -0.9357>}
 middle_03_l {parent=middle_02_l, loc=<2.6077 0.0534 0.1373>, rot=<0.0082 0.0188 0.1708 -0.9851>}
 pinky_metacarpal_l {parent=hand_l, loc=<1.7616 -0.9753 3.3814>, rot=<-0.0397 0.0814 -0.069 -0.9935>}
 pinky_01_l {parent=pinky_metacarpal_l, loc=<7.1709 -0.4831 -0.1094>, rot=<-0.0654 -0.0035 0.2242 -0.9723>}
 pinky_02_l {parent=pinky_01_l, loc=<2.8274 0.0893 -0.1043>, rot=<0.0037 0.0011 0.2854 -0.9584>}
 pinky_03_l {parent=pinky_02_l, loc=<2.2046 0.0307 -0.0078>, rot=<-0.0002 0.0098 0.1445 -0.9895>}
 ring_metacarpal_l {parent=hand_l, loc=<2.2074 -0.3625 1.3638>, rot=<-0.0284 0.0722 -0.0458 -0.9959>}
 ring_01_l {parent=ring_metacarpal_l, loc=<7.0295 -0.6197 -0.1446>, rot=<-0.0489 -0.0252 0.1892 -0.9804>}
 ring_02_l {parent=ring_01_l, loc=<3.6174 0.4647 -0.1027>, rot=<0 0 0.2806 -0.9598>}
 ring_03_l {parent=ring_02_l, loc=<2.4152 0.114 0.0199>, rot=<0 0 0.1189 -0.9929>}
 thumb_01_l {parent=hand_l, loc=<1.3118 -0.3191 -2.459>, rot=<0.6546 -0.4042 0.0664 -0.6354>}
 thumb_02_l {parent=thumb_01_l, loc=<5.9336 -0.0448 0.0401>, rot=<-0.0037 0.0213 0.298 -0.9543>}
 thumb_03_l {parent=thumb_02_l, loc=<2.2223 0.3682 -0.0907>, rot=<0.0118 -0.013 0.0066 -0.9998>}
 lowerarm_twist_01_l {parent=lowerarm_l, loc=<17.8549 0 0>, rot=<0 0 0 1>}
 lowerarm_twist_02_l {parent=lowerarm_l, loc=<9.6199 0 0>, rot=<0 0 0 1>}
 upperarm_twist_01_l {parent=upperarm_l, loc=<0.0041 0 0>, rot=<0 0 0 1>}
 upperarm_twist_02_l {parent=upperarm_l, loc=<13.7941 0 0>, rot=<0 0 0 1>}
 dyn_shoulder_pad_00 {parent=upperarm_l, loc=<-0.2925 -0.5779 6.8715>, rot=<-0.0134 0.3555 -0.014 -0.9345>}
 dyn_shoulder_pad_00_b {parent=dyn_shoulder_pad_00, loc=<0 0 0>, rot=<0 0 0 1>}
 dyn_shoulder_pad_02 {parent=upperarm_l, loc=<9.3264 -0.2284 4.688>, rot=<-0.0134 0.3555 -0.014 -0.9345>}
 dyn_shoulder_pad_03 {parent=upperarm_l, loc=<11.1216 -0.6661 3.0454>, rot=<-0.0134 0.3555 -0.014 -0.9345>}
 dyn_shoulder_pad_01 {parent=upperarm_l, loc=<2.2863 0.2912 5.9051>, rot=<-0.0134 0.3555 -0.014 -0.9345>}
 clavicle_r {parent=spine_05, loc=<4.3575 -0.1466 2.3223>, rot=<0.7289 0.0465 -0.6829 0.009>}
 upperarm_r {parent=clavicle_r, loc=<-15.5111 0 0>, rot=<0.0226 -0.3234 -0.0088 -0.9459>}
 lowerarm_r {parent=upperarm_r, loc=<-26.6515 0 0>, rot=<0.0391 -0.0099 -0.2453 -0.9686>}
 hand_r {parent=lowerarm_r, loc=<-22.3492 0 0>, rot=<-0.7081 -0.0507 0.0975 -0.6975>}
 index_metacarpal_r {parent=hand_r, loc=<-1.6534 -0.0181 2.9601>, rot=<0.0158 -0.0124 -0.0173 -0.9996>}
 index_01_r {parent=index_metacarpal_r, loc=<-8.3008 -0.0357 0.0524>, rot=<-0.0044 0.0485 0.0924 -0.9945>}
 index_02_r {parent=index_01_r, loc=<-3.3589 0.3003 -0.0453>, rot=<-0.002 0.0027 0.3225 -0.9466>}
 index_03_r {parent=index_02_r, loc=<-2.5578 0.1231 0.0251>, rot=<-0.002 0.0067 0.2558 -0.9667>}
 middle_metacarpal_r {parent=hand_r, loc=<-2.1856 0.1726 0.5984>, rot=<0.006 0.0168 -0.0329 -0.9993>}
 middle_01_r {parent=middle_metacarpal_r, loc=<-7.3273 0.2447 -0.0129>, rot=<-0.0208 0.0042 0.0858 -0.9961>}
 middle_02_r {parent=middle_01_r, loc=<-4.0142 -0.1411 -0.07>, rot=<-0.0217 0.0259 0.3512 -0.9357>}
 middle_03_r {parent=middle_02_r, loc=<-2.6074 -0.0534 -0.1312>, rot=<0.0083 0.0188 0.1708 -0.9851>}
 pinky_metacarpal_r {parent=hand_r, loc=<-1.7617 0.9753 -3.3814>, rot=<-0.0397 0.0814 -0.069 -0.9935>}
 pinky_01_r {parent=pinky_metacarpal_r, loc=<-7.1707 0.4832 0.1094>, rot=<-0.0654 -0.0035 0.2242 -0.9723>}
 pinky_02_r {parent=pinky_01_r, loc=<-2.8274 -0.0892 0.1043>, rot=<0.0037 0.0011 0.2854 -0.9584>}
 pinky_03_r {parent=pinky_02_r, loc=<-2.2045 -0.031 0.0078>, rot=<-0.0002 0.0098 0.1445 -0.9895>}
 ring_metacarpal_r {parent=hand_r, loc=<-2.2073 0.3627 -1.3638>, rot=<-0.0284 0.0722 -0.0458 -0.9959>}
 ring_01_r {parent=ring_metacarpal_r, loc=<-7.0297 0.6197 0.1446>, rot=<-0.0489 -0.0252 0.1892 -0.9804>}
 ring_02_r {parent=ring_01_r, loc=<-3.6171 -0.4648 0.1027>, rot=<0 0 0.2806 -0.9598>}
 ring_03_r {parent=ring_02_r, loc=<-2.4153 -0.1141 -0.0199>, rot=<0 0 0.1189 -0.9929>}
 thumb_01_r {parent=hand_r, loc=<-1.3119 0.3192 2.459>, rot=<0.6546 -0.4042 0.0664 -0.6354>}
 thumb_02_r {parent=thumb_01_r, loc=<-5.9335 0.0447 -0.0401>, rot=<-0.0037 0.0213 0.298 -0.9543>}
 thumb_03_r {parent=thumb_02_r, loc=<-2.2223 -0.3682 0.0907>, rot=<0.0118 -0.013 0.0066 -0.9998>}
 lowerarm_twist_01_r {parent=lowerarm_r, loc=<-17.855 0 0>, rot=<0 0 0 1>}
 lowerarm_twist_02_r {parent=lowerarm_r, loc=<-9.6199 0 0>, rot=<0 0 0 1>}
 upperarm_twist_01_r {parent=upperarm_r, loc=<-0.0041 0 0>, rot=<0 0 0 1>}
 upperarm_twist_02_r {parent=upperarm_r, loc=<-13.7942 0 0>, rot=<0 0 0 1>}
 neck_01 {parent=spine_05, loc=<11.3138 0.1529 0>, rot=<0 0 -0.2003 -0.9797>}
 neck_02 {parent=neck_01, loc=<4.5627 0 0>, rot=<0 0 0.0504 -0.9987>}
 head {parent=neck_02, loc=<5.3914 0 0>, rot=<0 0 0.115 -0.9934>}
 dyn_chest_dagger_00_base {parent=spine_05, loc=<-10.023 15.2222 1.834>, rot=<0.0284 -0.7065 -0.0284 -0.7065>}
 dyn_chest_dagger_00 {parent=dyn_chest_dagger_00_base, loc=<0 0 0>, rot=<0 0 0 1>}
 dyn_chest_dagger_01_base {parent=spine_05, loc=<-8.3447 14.8457 -3.0488>, rot=<0.0284 -0.7065 -0.0284 -0.7065>}
 dyn_chest_dagger_01 {parent=dyn_chest_dagger_01_base, loc=<0 0 0>, rot=<0 0 0 1>}
 dyn_chest_dagger_02_base {parent=spine_05, loc=<-6.6774 13.1707 -7.6282>, rot=<0.0284 -0.7065 -0.0284 -0.7065>}
 dyn_chest_dagger_02 {parent=dyn_chest_dagger_02_base, loc=<0 0 0>, rot=<0 0 0 1>}
 thigh_l {parent=pelvis, loc=<-2.5693 -0.0109 -7.7501>, rot=<0.0012 0.0572 0.0155 -0.9982>}
 calf_l {parent=thigh_l, loc=<-39.0484 0 0>, rot=<0.0006 -0.0161 -0.0391 -0.9991>}
 calf_twist_01_l {parent=calf_l, loc=<-35.813 -0.0967 0.0397>, rot=<0.0237 0.0095 -0.0035 -0.9997>}
 calf_twist_02_l {parent=calf_l, loc=<-18.8064 -0.0967 0.0397>, rot=<0.0237 0.0095 -0.0035 -0.9997>}
 foot_l {parent=calf_l, loc=<-42.561 0.1077 0.4292>, rot=<0.0418 -0.0455 0.0343 -0.9975>}
 ball_l {parent=foot_l, loc=<-6.1599 13.8264 0.6806>, rot=<0.0001 -0.0001 -0.7063 -0.7079>}
 thigh_twist_01_l {parent=thigh_l, loc=<-21.6303 0 0>, rot=<0.0013 0 0 -1>}
 thigh_r {parent=pelvis, loc=<-2.5691 -0.0109 7.7501>, rot=<0.0572 -0.0012 0.9982 0.0155>}
 calf_r {parent=thigh_r, loc=<39.0487 0 0>, rot=<0.0006 -0.0161 -0.0391 -0.9991>}
 calf_twist_01_r {parent=calf_r, loc=<35.8132 0.0967 -0.0397>, rot=<0.0237 0.0095 -0.0035 -0.9997>}
 calf_twist_02_r {parent=calf_r, loc=<18.8064 0.0967 -0.0397>, rot=<0.0237 0.0095 -0.0035 -0.9997>}
 dyn_dagger {parent=calf_twist_02_r, loc=<8.151 2.1092 6.4636>, rot=<-0.6688 0.0123 0.7417 0.0498>}
 foot_r {parent=calf_r, loc=<42.5612 -0.1077 -0.4292>, rot=<-0.0031 -0.0411 0.0414 -0.9983>}
 ball_r {parent=foot_r, loc=<6.1598 -13.8264 -0.6806>, rot=<0.0001 -0.0001 -0.7063 -0.7079>}
 thigh_twist_01_r {parent=thigh_r, loc=<21.6303 0 0>, rot=<0.0013 0 0 -1>}
 dyn_skirt_front_01_r {parent=pelvis, loc=<-0.9586 7.8629 11.415>, rot=<0.1318 -0.603 0.779 0.1098>}
 dyn_skirt_front_02_r {parent=dyn_skirt_front_01_r, loc=<8.5343 0 0>, rot=<0 0 0 1>}
 dyn_skirt_front_03_r {parent=dyn_skirt_front_02_r, loc=<8.6477 0 0>, rot=<0 0 0 1>}
 dyn_skirt_front_04_r {parent=dyn_skirt_front_03_r, loc=<10.2841 0 0>, rot=<0 0 0 1>}
 dyn_skirt_front_05_r {parent=dyn_skirt_front_04_r, loc=<8.3964 0 0>, rot=<0 0 0 1>}
 dyn_skirt_side_01_r {parent=pelvis, loc=<0.6301 -0.9921 14.0907>, rot=<0.094 -0.747 0.6525 0.0862>}
 dyn_skirt_side_02_r {parent=dyn_skirt_side_01_r, loc=<9.5387 0 0>, rot=<0 0 0 1>}
 dyn_skirt_side_03_r {parent=dyn_skirt_side_02_r, loc=<9.4128 0 0>, rot=<0 0 0 1>}
 dyn_skirt_side_04_r {parent=dyn_skirt_side_03_r, loc=<9.8165 0 0>, rot=<0 0 0 1>}
 dyn_skirt_side_05_r {parent=dyn_skirt_side_04_r, loc=<8.3964 0 0>, rot=<0 0 0 1>}
 dyn_skirt_mid_01_r {parent=pelvis, loc=<0.2096 -6.9034 11.7152>, rot=<0.1117 -0.9194 0.372 0.0615>}
 dyn_skirt_mid_02_r {parent=dyn_skirt_mid_01_r, loc=<9.0953 0 0>, rot=<0 0 0 1>}
 dyn_skirt_mid_03_r {parent=dyn_skirt_mid_02_r, loc=<9.163 0 0>, rot=<0 0 0 1>}
 dyn_skirt_mid_04_r {parent=dyn_skirt_mid_03_r, loc=<9.7725 0 0>, rot=<0 0 0 1>}
 dyn_skirt_mid_05_r {parent=dyn_skirt_mid_04_r, loc=<8.3964 0 0>, rot=<0 0 0 1>}
 dyn_skirt_back_01_r {parent=pelvis, loc=<0.4352 -10.8816 5.9619>, rot=<0.105 -0.9824 0.1521 0.0288>}
 dyn_skirt_back_02_r {parent=dyn_skirt_back_01_r, loc=<9.6105 0 0>, rot=<0 0 0 1>}
 dyn_skirt_back_03_r {parent=dyn_skirt_back_02_r, loc=<9.4756 0 0>, rot=<0 0 0 1>}
 dyn_skirt_back_04_r {parent=dyn_skirt_back_03_r, loc=<9.523 0 0>, rot=<0 0 0 1>}
 dyn_skirt_back_05_r {parent=dyn_skirt_back_04_r, loc=<8.3964 0 0>, rot=<0 0 0 1>}
 dyn_skirt_front_01_l {parent=pelvis, loc=<-0.4783 8.5306 -11.0447>, rot=<-0.1318 0.603 0.779 0.1098>}
 dyn_skirt_front_02_l {parent=dyn_skirt_front_01_l, loc=<9.6584 0 0>, rot=<0 0 0 1>}
 dyn_skirt_front_03_l {parent=dyn_skirt_front_02_l, loc=<8.2222 0 0>, rot=<0 0 0 1>}
 dyn_skirt_front_04_l {parent=dyn_skirt_front_03_l, loc=<9.6847 0 0>, rot=<0 0 0 1>}
 dyn_skirt_front_05_l {parent=dyn_skirt_front_04_l, loc=<8.3964 0 0>, rot=<0 0 0 1>}
 dyn_skirt_side_01_l {parent=pelvis, loc=<0.4251 -0.7297 -15.669>, rot=<0.1062 -0.765 -0.6261 -0.1077>}
 dyn_skirt_side_02_l {parent=dyn_skirt_side_01_l, loc=<9.5387 0 0>, rot=<0 0 0 1>}
 dyn_skirt_side_03_l {parent=dyn_skirt_side_02_l, loc=<9.4128 0 0>, rot=<0 0 0 1>}
 dyn_skirt_side_04_l {parent=dyn_skirt_side_03_l, loc=<9.8164 0 0>, rot=<0 0 0 1>}
 dyn_skirt_side_05_l {parent=dyn_skirt_side_04_l, loc=<8.3964 0 0>, rot=<0 0 0 1>}
 dyn_skirt_mid_01_l {parent=pelvis, loc=<0.1108 -6.7921 -11.6108>, rot=<0.1265 -0.9175 -0.371 -0.0675>}
 dyn_skirt_mid_02_l {parent=dyn_skirt_mid_01_l, loc=<9.0571 0 0>, rot=<0 0 0 1>}
 dyn_skirt_mid_03_l {parent=dyn_skirt_mid_02_l, loc=<9.0702 0 -0.0001>, rot=<0 0 0 1>}
 dyn_skirt_mid_04_l {parent=dyn_skirt_mid_03_l, loc=<8.8267 0 0.0001>, rot=<0 0 0 1>}
 dyn_skirt_mid_05_l {parent=dyn_skirt_mid_04_l, loc=<8.3964 0 0>, rot=<0 0 0 1>}
 dyn_skirt_back_01_l {parent=pelvis, loc=<0.3864 -10.929 -6.3724>, rot=<0.1345 -0.9787 -0.1512 -0.0336>}
 dyn_skirt_back_02_l {parent=dyn_skirt_back_01_l, loc=<9.4797 0 0>, rot=<0 0 0 1>}
 dyn_skirt_back_03_l {parent=dyn_skirt_back_02_l, loc=<9.191 -0.0001 0.0001>, rot=<0 0 0 1>}
 dyn_skirt_back_04_l {parent=dyn_skirt_back_03_l, loc=<9.2112 0 -0.0001>, rot=<0 0 0 1>}
 dyn_skirt_back_05_l {parent=dyn_skirt_back_04_l, loc=<8.3964 0 0>, rot=<0 0 0 1>}
 dyn_potion_MSTR {parent=pelvis, loc=<7.3715 -9.0901 11.5464>, rot=<0.0263 -0.3763 0.9261 0.0032>}
 dyn_potion {parent=dyn_potion_MSTR, loc=<0 0 0>, rot=<0 0 0 1>}
 dyn_bag_l_MSTR {parent=pelvis, loc=<10.6083 6.5699 -12.9429>, rot=<-0.0173 -0.9641 -0.265 -0.0047>}
 dyn_bag_l {parent=dyn_bag_l_MSTR, loc=<0 0 0>, rot=<0 0 0 1>}
 dyn_belt_knife_MSTR {parent=pelvis, loc=<10.4906 -2.4558 -15.314>, rot=<-0.0179 -0.9998 0 0>}
 dyn_belt_knife {parent=dyn_belt_knife_MSTR, loc=<0 0 0>, rot=<0 0 0 1>}
 dyn_bag_r_MSTR {parent=pelvis, loc=<8.8323 -2.049 15.1304>, rot=<-0.0127 -0.707 0.0127 -0.707>}
 dyn_bag_r {parent=dyn_bag_r_MSTR, loc=<0 0 0>, rot=<-0.6774 0 0.7357 0>}
 dyn_bag_MSTR {parent=pelvis, loc=<8.3905 8.1747 13.4551>, rot=<0.034 0.2359 0.9712 0.0051>}
 dyn_bag {parent=dyn_bag_MSTR, loc=<0 0 0>, rot=<0 0 0 1>}
]
variables [
 BaseBall_Slide {type=double, default=0}
 CrouchOnly {type=bool, default=False}
 CrouchTarget {type=bool, default=False}
 Is Relaxed Level 1 {type=bool, default=False}
 Is Sky Diving {type=bool, default=False}
 Is Sky Diving_0 {type=bool, default=False}
 Is Sky Diving_1 {type=bool, default=False}
 Is Targeting {type=bool, default=False}
 Knee_Slide {type=double, default=0}
 MovingTarget {type=bool, default=False}
 SkydiveFloat {type=bool, default=False}
 SkydiveFloat_0 {type=bool, default=False}
]
