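# Post-process style AnimBlueprint for the male Swamp Knight Commando body, adding secondary motion to skirt, bags, potion, belt knife and chest daggers.
# Several parallel chains copy the attached parent mesh's pose, with variants using skirt or crouch physics assets; the graph also reads bIsMoving2D.
# AnimDynamics swings accessories, while pelvis-space RigidBody nodes simulate physics-asset bodies.
# Weighted constraints tie skirt and dagger bases to body bones, with additive bone offsets correcting placement.
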AnimBlueprint /Game/Characters/Player/Male/Medium/Bodies/M_MED_Commando_SwampKnight/Meshes/Swamp_Knight_Commando_Skeleton_AnimBlueprint
target_skeleton=&/Swamp_Knight_Commando_Skeleton
graphs [
 AnimGraph {nodes [
   n1 = Anim Dynamics - Bone: dyn_potion(ComponentPose=Local To Component(LocalPose=Copy Pose From Mesh() {bUseAttachedParent=true}), Alpha=0.5, ExternalForce="0,0,0") {
    angularDampingOverride=1
    angularSpringConstant=20000
    bAngularSpring=true
    bOverrideAngularDamping=true
    bUseGravityOverride=true
    gravityScale=0.35
    simulationSpace=RootRelative
    alphaBoolBlend {blendOption=Linear}
    boundBone {boneName=dyn_potion}
    componentLinearAccScale=<1 1 1>
    componentLinearVelScale=<1 1 1>
    gravityOverride {z=-980}
    physicsBodyDefinitions [{
      boundBone {boneName=dyn_potion}
      boxExtents=<15 10 10>
      constraintSetup {angularLimitsMax=<10 5 15>, angularLimitsMin=<-10 -10 -5>, angularTarget=<0 0 0>}
      localJointOffset=<5 0 0>
     }]
    relativeSpaceBone {boneName=spine_05}
   }
   n2 = Anim Dynamics - Bone: dyn_bag_r(ComponentPose=@n1, Alpha=0.5, ExternalForce="0,0,0") {
    angularDampingOverride=1
    angularSpringConstant=10000
    bAngularSpring=true
    bOverrideAngularDamping=true
    bUseGravityOverride=true
    gravityScale=0.35
    simulationSpace=RootRelative
    alphaBoolBlend {blendOption=Linear}
    boundBone {boneName=dyn_bag_r}
    componentLinearAccScale=<1 1 1>
    componentLinearVelScale=<1 1 1>
    gravityOverride {z=-980}
    physicsBodyDefinitions [{
      boundBone {boneName=dyn_bag_r}
      boxExtents=<15 15 15>
      constraintSetup {angularLimitsMax=<5 10 2>, angularLimitsMin=<-5 -15 -2>, angularTarget=<0 0 0>}
      localJointOffset=<3 0 0>
     }]
    relativeSpaceBone {boneName=pelvis}
   }
   n3 = Anim Dynamics - Bone: dyn_belt_knife(ComponentPose=@n2, Alpha=0.3, ExternalForce="0,0,0") {
    angularDampingOverride=1
    angularSpringConstant=20000
    bAngularSpring=true
    bOverrideAngularDamping=true
    bUseGravityOverride=true
    gravityScale=0.35
    simulationSpace=RootRelative
    alphaBoolBlend {blendOption=Linear}
    boundBone {boneName=dyn_belt_knife}
    componentLinearAccScale=<1 1 1>
    componentLinearVelScale=<1 1 1>
    gravityOverride {z=-980}
    physicsBodyDefinitions [{
      boundBone {boneName=dyn_belt_knife}
      boxExtents=<10 10 10>
      constraintSetup {angularLimitsMax=<0 1 5>, angularLimitsMin=<0 -10 -5>, angularTarget=<0 0 0>}
      localJointOffset=<5 0 0>
     }]
    relativeSpaceBone {boneName=pelvis}
   }
   n4 = Anim Dynamics - Bone: dyn_bag_l(ComponentPose=@n3, Alpha=0.4, ExternalForce="0,0,0") {
    angularDampingOverride=0.9999
    angularSpringConstant=5000
    bAngularSpring=true
    bOverrideAngularDamping=true
    bUseGravityOverride=true
    gravityScale=0.35
    simulationSpace=RootRelative
    alphaBoolBlend {blendOption=Linear}
    boundBone {boneName=dyn_bag_l}
    componentLinearAccScale=<1 1 1>
    componentLinearVelScale=<1 1 1>
    gravityOverride {z=-980}
    physicsBodyDefinitions [{
      boundBone {boneName=dyn_bag_l}
      boxExtents=<10 10 10>
      constraintSetup {angularLimitsMax=<1 10 1>, angularLimitsMin=<-1 -10 -1>, angularTarget=<0 0 0>}
      localJointOffset=<5 0 0>
     }]
   }
   n5 = "Constraint - dyn_skirt_front_01_l (P)"(ComponentPose=@n4, Alpha=0.8, ConstraintWeights_0=1) {
    alphaBoolBlend {blendOption=Linear}
    boneToModify {boneName=dyn_skirt_front_01_l}
    constraintSetup [{targetBone {boneName=thigh_l}}]
    constraintWeights [1]
   }
   n6 = "Transform (Modify) Bone - Bone: dyn_bag_l_MSTR"(ComponentPose=@n5, Alpha=1, Translation="2.645,1.657,0.000", Rotation="-3.676,-25.280,-31.323", Scale="1,1,1") {
    rotationMode=BMM_Additive
    translationMode=BMM_Additive
    alphaBoolBlend {blendOption=Linear}
    boneToModify {boneName=dyn_bag_l_MSTR}
    rotation=<rot -3.676 -25.28 -31.323>
    translation=<2.645 1.657>
   }
   n7 = "Transform (Modify) Bone - Bone: dyn_belt_knife"(ComponentPose=@n6, Alpha=1, Translation="-0.014,-4.510,0.000", Rotation="-2.564,-12.965,-17.277", Scale="1,1,1") {
    rotationMode=BMM_Additive
    translationMode=BMM_Additive
    alphaBoolBlend {blendOption=Linear}
    boneToModify {boneName=dyn_belt_knife}
    rotation=<rot -2.564 -12.965 -17.277>
    translation=<-0.014 -4.51>
   }
   n8 = "Transform (Modify) Bone - Bone: dyn_bag"(ComponentPose=@n7, Alpha=1, Translation="0,0,5", Rotation="0,0,0", Scale="1,1,1") {
    rotationMode=BMM_Additive
    translationMode=BMM_Additive
    alphaBoolBlend {blendOption=Linear}
    boneToModify {boneName=dyn_bag}
    translation {z=5}
   }
   n9 = "Constraint - dyn_skirt_back_04_r (PPPPP)"!(
    ComponentPose=@n8
    Alpha=1
    ConstraintWeights_0=0
    ConstraintWeights_1:float=0
    ConstraintWeights_2:float=0.0003
    ConstraintWeights_3:float=0
    ConstraintWeights_4:float=0.9997
   ) -> (Pose) {
    alphaBoolBlend {blendOption=Linear}
    boneToModify {boneName=dyn_skirt_back_04_r}
    constraintSetup [
     {targetBone {boneName=ball_r}}
     {targetBone {boneName=calf_twist_02_r}}
     {targetBone {boneName=calf_twist_01_r}}
     {targetBone {boneName=calf_r}}
     {targetBone {boneName=foot_r}}
    ]
    constraintWeights [1,1,1,1,1]
   }
   n10 = "Constraint - dyn_skirt_side_04_r (P)"!(ComponentPose=@n9, Alpha=1, ConstraintWeights_0=1, ConstraintWeights_1:float=0) -> (Pose) {
    alphaBoolBlend {blendOption=Linear}
    boneToModify {boneName=dyn_skirt_side_04_r}
    constraintSetup [{targetBone {boneName=dyn_dagger}}]
    constraintWeights [1,0]
   }
   n11 = "Constraint - dyn_skirt_front_04_r (PPPPPP)"!(
    ComponentPose=@n10
    Alpha=1
    ConstraintWeights_0=0.0001
    ConstraintWeights_1:float=0.4781
    ConstraintWeights_2:float=0
    ConstraintWeights_3:float=0.0003
    ConstraintWeights_4:float=0.5194
    ConstraintWeights_5:float=0.0021
   ) -> (Pose) {
    alphaBoolBlend {blendOption=Linear}
    boneToModify {boneName=dyn_skirt_front_04_r}
    constraintSetup [
     {targetBone {boneName=spine_01}}
     {targetBone {boneName=thigh_r}}
     {targetBone {boneName=calf_twist_02_r}}
     {targetBone {boneName=calf_r}}
     {targetBone {boneName=thigh_twist_01_r}}
     {targetBone {boneName=pelvis}}
    ]
    constraintWeights [1,1,1,1,1,1]
   }
   n12 = "Constraint - dyn_skirt_front_02_r (PPPPPP)"!(
    ComponentPose=@n11
    Alpha=1
    ConstraintWeights_0=0.0287
    ConstraintWeights_1:float=0.0225
    ConstraintWeights_2:float=0.0197
    ConstraintWeights_3:float=0.6174
    ConstraintWeights_4:float=0.04
    ConstraintWeights_5:float=0.2717
   ) -> (Pose) {
    alphaBoolBlend {blendOption=Linear}
    boneToModify {boneName=dyn_skirt_front_02_r}
    constraintSetup [
     {targetBone {boneName=spine_01}}
     {targetBone {boneName=dyn_skirt_front_01_r}}
     {targetBone {boneName=thigh_r}}
     {targetBone {boneName=thigh_twist_01_r}}
     {targetBone {boneName=thigh_twist_01_l}}
     {targetBone {boneName=pelvis}}
    ]
    constraintWeights [1,1,1,1,1,1]
   }
   n13 = "Constraint - dyn_skirt_front_03_r (PPPP)"!(
    ComponentPose=@n12
    Alpha=1
    ConstraintWeights_0=0.0054
    ConstraintWeights_1:float=0.1072
    ConstraintWeights_2:float=0.8096
    ConstraintWeights_3:float=0.0777
   ) -> (Pose) {
    alphaBoolBlend {blendOption=Linear}
    boneToModify {boneName=dyn_skirt_front_03_r}
    constraintSetup [{targetBone {boneName=spine_01}}, {targetBone {boneName=thigh_r}}, {targetBone {boneName=thigh_twist_01_r}}, {targetBone {boneName=pelvis}}]
    constraintWeights [1,1,1,1]
   }
   n14 = "Constraint - dyn_skirt_front_01_r (P)"(ComponentPose=@n13, Alpha=6, ConstraintWeights_0=1) {
    alphaBoolBlend {blendOption=Linear}
    boneToModify {boneName=dyn_skirt_front_01_r}
    constraintSetup [{targetBone {boneName=thigh_r}}]
    constraintWeights [1]
   }
   n15 = RigidBody(ComponentPose=@n14, ExternalForce=Get GravityOverride()) {
    bClampLinearTranslationLimitToRefPose=true
    overridePhysicsAsset=&/M_MED_Swamp_Knight_Skirt_Physics
    simulationSpace=BaseBoneSpace
    alphaBoolBlend {blendOption=Linear}
    baseBoneRef {boneName=pelvis}
    componentAppliedLinearAccClamp=<2750 2750 2750>
    componentLinearAccScale=<0.25 0.25 0.15>
    componentLinearVelScale=<0.25 0.25 0.15>
    overrideWorldGravity {z=-980}
   }
   n16 = "Transform (Modify) Bone - Bone: dyn_skirt_mid_01_r"(ComponentPose=@n15, Alpha=1, Translation="0,0,0", Rotation="0.000,30.000,-0.000", Scale="1,1,1") {
    rotationMode=BMM_Additive
    translationMode=BMM_Additive
    alphaBoolBlend {blendOption=Linear}
    boneToModify {boneName=dyn_skirt_mid_01_r}
    rotation {yaw=30}
   }
   n17 = "Transform (Modify) Bone - Bone: dyn_skirt_side_01_r"(ComponentPose=@n16, Alpha=1, Translation="0,0,0", Rotation="15.074,-9.729,0.898", Scale="1,1,1") {
    rotationMode=BMM_Additive
    translationMode=BMM_Additive
    alphaBoolBlend {blendOption=Linear}
    boneToModify {boneName=dyn_skirt_side_01_r}
    rotation=<rot 15.074 -9.729 0.898>
   }
   n18 = "Transform (Modify) Bone - Bone: dyn_skirt_front_01_r"(ComponentPose=@n17, Alpha=1, Translation="-6.25,-1,0", Rotation="5.000,0.000,0.000", Scale="1,1,1") {
    rotationMode=BMM_Additive
    translationMode=BMM_Additive
    alphaBoolBlend {blendOption=Linear}
    boneToModify {boneName=dyn_skirt_front_01_r}
    rotation {pitch=5}
    translation=<-6.25 -1>
   }
   n19 = "Constraint - dyn_chest_dagger_00_base (PPPPPPPP)"!(
    ComponentPose=@n18
    Alpha=1
    ConstraintWeights_0=0.006
    ConstraintWeights_1:float=0.0476
    ConstraintWeights_2:float=0.187
    ConstraintWeights_3:float=0.3862
    ConstraintWeights_4:float=0.2588
    ConstraintWeights_5:float=0.0289
    ConstraintWeights_6:float=0.0831
    ConstraintWeights_7:float=0.0022
   ) -> (Pose) {
    alphaBoolBlend {blendOption=Linear}
    boneToModify {boneName=dyn_chest_dagger_00_base}
    constraintSetup [
     {targetBone {boneName=clavicle_l}}
     {targetBone {boneName=spine_01}}
     {targetBone {boneName=spine_02}}
     {targetBone {boneName=spine_03}}
     {targetBone {boneName=spine_04}}
     {targetBone {boneName=spine_05}}
     {targetBone {boneName=clavicle_r}}
     {targetBone {boneName=pelvis}}
    ]
    constraintWeights [1,1,1,1,1,1,1,1]
   }
   n20 = "Constraint - dyn_chest_dagger_01_base (PPPPPPPP)"!(
    ComponentPose=@n19
    Alpha=1
    ConstraintWeights_0=0.1183
    ConstraintWeights_1:float=0.0377
    ConstraintWeights_2:float=0.1683
    ConstraintWeights_3:float=0.3301
    ConstraintWeights_4:float=0.2998
    ConstraintWeights_5:float=0.0458
    ConstraintWeights_6:float=0
    ConstraintWeights_7:float=0
   ) -> (Pose) {
    alphaBoolBlend {blendOption=Linear}
    boneToModify {boneName=dyn_chest_dagger_01_base}
    constraintSetup [
     {targetBone {boneName=clavicle_l}}
     {targetBone {boneName=spine_01}}
     {targetBone {boneName=spine_02}}
     {targetBone {boneName=spine_03}}
     {targetBone {boneName=spine_04}}
     {targetBone {boneName=spine_05}}
     {targetBone {boneName=clavicle_r}}
     {targetBone {boneName=pelvis}}
    ]
    constraintWeights [1,1,1,1,1,1,1,1]
   }
   n21 = "Constraint - dyn_chest_dagger_02_base (PPPPPPPP)"!(
    ComponentPose=@n20
    Alpha=1
    ConstraintWeights_0=0.1681
    ConstraintWeights_1:float=0.0351
    ConstraintWeights_2:float=0.1477
    ConstraintWeights_3:float=0.2823
    ConstraintWeights_4:float=0.3184
    ConstraintWeights_5:float=0.0519
    ConstraintWeights_6:float=0
    ConstraintWeights_7:float=0.0001
   ) -> (Pose) {
    alphaBoolBlend {blendOption=Linear}
    boneToModify {boneName=dyn_chest_dagger_02_base}
    constraintSetup [
     {targetBone {boneName=clavicle_l}}
     {targetBone {boneName=spine_01}}
     {targetBone {boneName=spine_02}}
     {targetBone {boneName=spine_03}}
     {targetBone {boneName=spine_04}}
     {targetBone {boneName=spine_05}}
     {targetBone {boneName=clavicle_r}}
     {targetBone {boneName=pelvis}}
    ]
    constraintWeights [1,1,1,1,1,1,1,1]
   }
   n22 = Anim Dynamics - Bone: dyn_chest_dagger_00(ComponentPose=@n21, Alpha=0.6, ExternalForce="0,0,0") {
    angularDampingOverride=5
    bGravityOverrideInSimSpace=true
    bUseGravityOverride=true
    alphaBoolBlend {blendOption=Linear}
    boundBone {boneName=dyn_chest_dagger_00}
    gravityOverride {z=-980}
    physicsBodyDefinitions [{
      sphereCollisionRadius=0
      boundBone {boneName=dyn_chest_dagger_00}
      boxExtents=<15 15 15>
      constraintSetup {twistAxis=AxisZ, angularLimitsMax=<1 1 4>, angularLimitsMin=<-1 -1 -4>, angularTarget=<0 0 0>}
      localJointOffset=<0 0 -1>
     }]
    relativeSpaceBone {boneName=root}
   }
   n23 = Anim Dynamics - Bone: dyn_chest_dagger_01(ComponentPose=@n22, Alpha=0.5, ExternalForce="0,0,0") {
    angularDampingOverride=5
    bGravityOverrideInSimSpace=true
    bUseGravityOverride=true
    alphaBoolBlend {blendOption=Linear}
    boundBone {boneName=dyn_chest_dagger_01}
    gravityOverride {z=-980}
    physicsBodyDefinitions [{
      sphereCollisionRadius=0
      boundBone {boneName=dyn_chest_dagger_01}
      boxExtents=<13 13 13>
      constraintSetup {twistAxis=AxisZ, angularLimitsMax=<4 4 0>, angularLimitsMin=<-4 -4 0>, angularTarget=<0 0 0>}
      localJointOffset=<0 0 -1>
     }]
    relativeSpaceBone {boneName=root}
   }
   n24 = Anim Dynamics - Bone: dyn_chest_dagger_02(ComponentPose=@n23, Alpha=0.4, ExternalForce="0,0,0") {
    angularDampingOverride=5
    bGravityOverrideInSimSpace=true
    bUseGravityOverride=true
    alphaBoolBlend {blendOption=Linear}
    boundBone {boneName=dyn_chest_dagger_02}
    gravityOverride {z=-980}
    physicsBodyDefinitions [{
      sphereCollisionRadius=0
      boundBone {boneName=dyn_chest_dagger_02}
      boxExtents=<15 15 15>
      constraintSetup {twistAxis=AxisZ, angularLimitsMax=<4 4 0>, angularLimitsMin=<-4 -4 0>, angularTarget=<0 0 0>}
      localJointOffset=<0 0 -1>
     }]
    relativeSpaceBone {boneName=root}
   }
   n25 = Component To Local(ComponentPose=@n24)
   n26 = Anim Dynamics - Bone: dyn_potion(ComponentPose=Local To Component(LocalPose=Copy Pose From Mesh() {bUseAttachedParent=true}), Alpha=0.5, ExternalForce="0,0,0") {
    angularDampingOverride=1
    angularSpringConstant=20000
    bAngularSpring=true
    bOverrideAngularDamping=true
    bUseGravityOverride=true
    gravityScale=0.35
    simulationSpace=RootRelative
    alphaBoolBlend {blendOption=Linear}
    boundBone {boneName=dyn_potion}
    componentLinearAccScale=<1 1 1>
    componentLinearVelScale=<1 1 1>
    gravityOverride {z=-980}
    physicsBodyDefinitions [{
      boundBone {boneName=dyn_potion}
      boxExtents=<15 10 10>
      constraintSetup {angularLimitsMax=<10 5 15>, angularLimitsMin=<-10 -10 -5>, angularTarget=<0 0 0>}
      localJointOffset=<5 0 0>
     }]
    relativeSpaceBone {boneName=spine_05}
   }
   n27 = Anim Dynamics - Bone: dyn_bag_r(ComponentPose=@n26, Alpha=0.5, ExternalForce="0,0,0") {
    angularDampingOverride=1
    angularSpringConstant=10000
    bAngularSpring=true
    bOverrideAngularDamping=true
    bUseGravityOverride=true
    gravityScale=0.35
    simulationSpace=RootRelative
    alphaBoolBlend {blendOption=Linear}
    boundBone {boneName=dyn_bag_r}
    componentLinearAccScale=<1 1 1>
    componentLinearVelScale=<1 1 1>
    gravityOverride {z=-980}
    physicsBodyDefinitions [{
      boundBone {boneName=dyn_bag_r}
      boxExtents=<15 15 15>
      constraintSetup {angularLimitsMax=<5 10 2>, angularLimitsMin=<-5 -15 -2>, angularTarget=<0 0 0>}
      localJointOffset=<3 0 0>
     }]
    relativeSpaceBone {boneName=pelvis}
   }
   n28 = Anim Dynamics - Bone: dyn_belt_knife(ComponentPose=@n27, Alpha=0.3, ExternalForce="0,0,0") {
    angularDampingOverride=1
    angularSpringConstant=20000
    bAngularSpring=true
    bOverrideAngularDamping=true
    bUseGravityOverride=true
    gravityScale=0.35
    simulationSpace=RootRelative
    alphaBoolBlend {blendOption=Linear}
    boundBone {boneName=dyn_belt_knife}
    componentLinearAccScale=<1 1 1>
    componentLinearVelScale=<1 1 1>
    gravityOverride {z=-980}
    physicsBodyDefinitions [{
      boundBone {boneName=dyn_belt_knife}
      boxExtents=<10 10 10>
      constraintSetup {angularLimitsMax=<0 1 5>, angularLimitsMin=<0 -10 -5>, angularTarget=<0 0 0>}
      localJointOffset=<5 0 0>
     }]
    relativeSpaceBone {boneName=pelvis}
   }
   n29 = Anim Dynamics - Bone: dyn_bag_l(ComponentPose=@n28, Alpha=0.4, ExternalForce="0,0,0") {
    angularDampingOverride=0.9999
    angularSpringConstant=5000
    bAngularSpring=true
    bOverrideAngularDamping=true
    bUseGravityOverride=true
    gravityScale=0.35
    simulationSpace=RootRelative
    alphaBoolBlend {blendOption=Linear}
    boundBone {boneName=dyn_bag_l}
    componentLinearAccScale=<1 1 1>
    componentLinearVelScale=<1 1 1>
    gravityOverride {z=-980}
    physicsBodyDefinitions [{
      boundBone {boneName=dyn_bag_l}
      boxExtents=<10 10 10>
      constraintSetup {angularLimitsMax=<1 10 1>, angularLimitsMin=<-1 -10 -1>, angularTarget=<0 0 0>}
      localJointOffset=<5 0 0>
     }]
   }
   n30 = "Transform (Modify) Bone - Bone: dyn_bag_l_MSTR"(ComponentPose=@n29, Alpha=1, Translation="2.645,1.657,0.000", Rotation="-3.676,-25.280,-31.323", Scale="1,1,1") {
    rotationMode=BMM_Additive
    translationMode=BMM_Additive
    alphaBoolBlend {blendOption=Linear}
    boneToModify {boneName=dyn_bag_l_MSTR}
    rotation=<rot -3.676 -25.28 -31.323>
    translation=<2.645 1.657>
   }
   n31 = "Transform (Modify) Bone - Bone: dyn_belt_knife"(ComponentPose=@n30, Alpha=1, Translation="-0.014,-4.510,0.000", Rotation="-2.564,-12.965,-17.277", Scale="1,1,1") {
    rotationMode=BMM_Additive
    translationMode=BMM_Additive
    alphaBoolBlend {blendOption=Linear}
    boneToModify {boneName=dyn_belt_knife}
    rotation=<rot -2.564 -12.965 -17.277>
    translation=<-0.014 -4.51>
   }
   n32 = "Transform (Modify) Bone - Bone: dyn_skirt_front_01_l"(ComponentPose=@n31, Alpha=1, Translation="11.440,-2.989,2.643", Rotation="-28.745,-53.347,-34.048", Scale="1,1,1") {
    rotationMode=BMM_Additive
    translationMode=BMM_Additive
    alphaBoolBlend {blendOption=Linear}
    boneToModify {boneName=dyn_skirt_front_01_l}
    rotation=<rot -28.745 -53.347 -34.048>
    translation=<11.44 -2.989 2.643>
   }
   n33 = "Transform (Modify) Bone - Bone: dyn_skirt_front_02_l"(ComponentPose=@n32, Alpha=1, Translation="0,0,0", Rotation="-0.000,0.000,15.000", Scale="1,1,1") {
    rotationMode=BMM_Additive
    translationMode=BMM_Additive
    alphaBoolBlend {blendOption=Linear}
    boneToModify {boneName=dyn_skirt_front_02_l}
    rotation {roll=15}
   }
   n34 = "Transform (Modify) Bone - Bone: dyn_skirt_side_02_l"(ComponentPose=@n33, Alpha=1, Translation="0,0,2", Rotation="30.000,0.000,-10.000", Scale="1,1,1") {
    rotationMode=BMM_Additive
    translationMode=BMM_Additive
    alphaBoolBlend {blendOption=Linear}
    boneToModify {boneName=dyn_skirt_side_02_l}
    rotation {pitch=30, roll=-10}
    translation {z=2}
   }
   n35 = "Transform (Modify) Bone - Bone: dyn_skirt_mid_01_l"(ComponentPose=@n34, Alpha=1, Translation="0,0,0", Rotation="0.000,0.000,-5.000", Scale="1,1,1") {
    rotationMode=BMM_Additive
    translationMode=BMM_Additive
    alphaBoolBlend {blendOption=Linear}
    boneToModify {boneName=dyn_skirt_mid_01_l}
    rotation {roll=-5}
   }
   n36 = "Transform (Modify) Bone - Bone: dyn_skirt_back_01_l"(ComponentPose=@n35, Alpha=1, Translation="0,0,0", Rotation="0,0,0", Scale="1,1,1") {rotationMode=BMM_Additive, translationMode=BMM_Additive, alphaBoolBlend {blendOption=Linear}, boneToModify {boneName=dyn_skirt_back_01_l}}
   n37 = "Transform (Modify) Bone - Bone: dyn_skirt_mid_02_l"(ComponentPose=@n36, Alpha=1, Translation="0,0,0", Rotation="0,0,0", Scale="1,1,1") {rotationMode=BMM_Additive, translationMode=BMM_Additive, alphaBoolBlend {blendOption=Linear}, boneToModify {boneName=dyn_skirt_mid_02_l}}
   n38 = "Constraint - dyn_skirt_back_04_r (PPPPP)"(
    ComponentPose=@n37
    Alpha=1
    ConstraintWeights_0=0
    ConstraintWeights_1=0
    ConstraintWeights_2=0.0003
    ConstraintWeights_3=0
    ConstraintWeights_4=0.9997
   ) {
    alphaBoolBlend {blendOption=Linear}
    boneToModify {boneName=dyn_skirt_back_04_r}
    constraintSetup [
     {targetBone {boneName=ball_r}}
     {targetBone {boneName=calf_twist_02_r}}
     {targetBone {boneName=calf_twist_01_r}}
     {targetBone {boneName=calf_r}}
     {targetBone {boneName=foot_r}}
    ]
    constraintWeights [1,1,1,1,1]
   }
   n39 = "Constraint - dyn_skirt_side_04_r (P)"(ComponentPose=@n38, Alpha=1, ConstraintWeights_0=1, ConstraintWeights_1=0) {
    alphaBoolBlend {blendOption=Linear}
    boneToModify {boneName=dyn_skirt_side_04_r}
    constraintSetup [{targetBone {boneName=dyn_dagger}}]
    constraintWeights [1,0]
   }
   n40 = "Constraint - dyn_skirt_front_04_r (PPPPPP)"(
    ComponentPose=@n39
    Alpha=1
    ConstraintWeights_0=0.0001
    ConstraintWeights_1=0.4781
    ConstraintWeights_2=0
    ConstraintWeights_3=0.0003
    ConstraintWeights_4=0.5194
    ConstraintWeights_5=0.0021
   ) {
    alphaBoolBlend {blendOption=Linear}
    boneToModify {boneName=dyn_skirt_front_04_r}
    constraintSetup [
     {targetBone {boneName=spine_01}}
     {targetBone {boneName=thigh_r}}
     {targetBone {boneName=calf_twist_02_r}}
     {targetBone {boneName=calf_r}}
     {targetBone {boneName=thigh_twist_01_r}}
     {targetBone {boneName=pelvis}}
    ]
    constraintWeights [1,1,1,1,1,1]
   }
   n41 = "Constraint - dyn_skirt_front_02_r (PPPPPP)"(
    ComponentPose=@n40
    Alpha=1
    ConstraintWeights_0=0.0287
    ConstraintWeights_1=0.0225
    ConstraintWeights_2=0.0197
    ConstraintWeights_3=0.6174
    ConstraintWeights_4=0.04
    ConstraintWeights_5=0.2717
   ) {
    alphaBoolBlend {blendOption=Linear}
    boneToModify {boneName=dyn_skirt_front_02_r}
    constraintSetup [
     {targetBone {boneName=spine_01}}
     {targetBone {boneName=dyn_skirt_front_01_r}}
     {targetBone {boneName=thigh_r}}
     {targetBone {boneName=thigh_twist_01_r}}
     {targetBone {boneName=thigh_twist_01_l}}
     {targetBone {boneName=pelvis}}
    ]
    constraintWeights [1,1,1,1,1,1]
   }
   n42 = "Constraint - dyn_skirt_front_03_r (PPPP)"(ComponentPose=@n41, Alpha=1, ConstraintWeights_0=0.0054, ConstraintWeights_1=0.1072, ConstraintWeights_2=0.8096, ConstraintWeights_3=0.0777) {
    alphaBoolBlend {blendOption=Linear}
    boneToModify {boneName=dyn_skirt_front_03_r}
    constraintSetup [{targetBone {boneName=spine_01}}, {targetBone {boneName=thigh_r}}, {targetBone {boneName=thigh_twist_01_r}}, {targetBone {boneName=pelvis}}]
    constraintWeights [1,1,1,1]
   }
   n43 = "Constraint - dyn_skirt_front_01_r (P)"(ComponentPose=@n42, Alpha=6, ConstraintWeights_0=1) {
    alphaBoolBlend {blendOption=Linear}
    boneToModify {boneName=dyn_skirt_front_01_r}
    constraintSetup [{targetBone {boneName=thigh_r}}]
    constraintWeights [1]
   }
   n44 = RigidBody(ComponentPose=@n43, ExternalForce=Get GravityOverride()) {
    bClampLinearTranslationLimitToRefPose=true
    overridePhysicsAsset=&/M_MED_Swamp_Knight_Crouch_Physics
    simulationSpace=BaseBoneSpace
    alphaBoolBlend {blendOption=Linear}
    baseBoneRef {boneName=pelvis}
    componentAppliedLinearAccClamp=<2500 2500 2500>
    componentLinearAccScale=<0.3 0.3 0.2>
    componentLinearVelScale=<0.3 0.3 0.2>
    overrideWorldGravity {z=-980}
   }
   n45 = "Transform (Modify) Bone - Bone: dyn_skirt_mid_01_r"(ComponentPose=@n44, Alpha=1, Translation="0,0,0", Rotation="0.000,30.000,-0.000", Scale="1,1,1") {
    rotationMode=BMM_Additive
    translationMode=BMM_Additive
    alphaBoolBlend {blendOption=Linear}
    boneToModify {boneName=dyn_skirt_mid_01_r}
    rotation {yaw=30}
   }
   n46 = "Transform (Modify) Bone - Bone: dyn_skirt_side_01_r"(ComponentPose=@n45, Alpha=1, Translation="0,0,0", Rotation="15.074,-9.729,0.898", Scale="1,1,1") {
    rotationMode=BMM_Additive
    translationMode=BMM_Additive
    alphaBoolBlend {blendOption=Linear}
    boneToModify {boneName=dyn_skirt_side_01_r}
    rotation=<rot 15.074 -9.729 0.898>
   }
   n47 = "Transform (Modify) Bone - Bone: dyn_skirt_front_01_r"(ComponentPose=@n46, Alpha=1, Translation="-5,0,0", Rotation="5.000,0.000,0.000", Scale="1,1,1") {
    rotationMode=BMM_Additive
    translationMode=BMM_Additive
    alphaBoolBlend {blendOption=Linear}
    boneToModify {boneName=dyn_skirt_front_01_r}
    rotation {pitch=5}
    translation {x=-5}
   }
   n48 = "Constraint - dyn_chest_dagger_00_base (PPPPPPPP)"(
    ComponentPose=@n47
    Alpha=1
    ConstraintWeights_0=0.006
    ConstraintWeights_1=0.0476
    ConstraintWeights_2=0.187
    ConstraintWeights_3=0.3862
    ConstraintWeights_4=0.2588
    ConstraintWeights_5=0.0289
    ConstraintWeights_6=0.0831
    ConstraintWeights_7=0.0022
   ) {
    alphaBoolBlend {blendOption=Linear}
    boneToModify {boneName=dyn_chest_dagger_00_base}
    constraintSetup [
     {targetBone {boneName=clavicle_l}}
     {targetBone {boneName=spine_01}}
     {targetBone {boneName=spine_02}}
     {targetBone {boneName=spine_03}}
     {targetBone {boneName=spine_04}}
     {targetBone {boneName=spine_05}}
     {targetBone {boneName=clavicle_r}}
     {targetBone {boneName=pelvis}}
    ]
    constraintWeights [1,1,1,1,1,1,1,1]
   }
   n49 = "Constraint - dyn_chest_dagger_01_base (PPPPPPPP)"(
    ComponentPose=@n48
    Alpha=1
    ConstraintWeights_0=0.1183
    ConstraintWeights_1=0.0377
    ConstraintWeights_2=0.1683
    ConstraintWeights_3=0.3301
    ConstraintWeights_4=0.2998
    ConstraintWeights_5=0.0458
    ConstraintWeights_6=0
    ConstraintWeights_7=0
   ) {
    alphaBoolBlend {blendOption=Linear}
    boneToModify {boneName=dyn_chest_dagger_01_base}
    constraintSetup [
     {targetBone {boneName=clavicle_l}}
     {targetBone {boneName=spine_01}}
     {targetBone {boneName=spine_02}}
     {targetBone {boneName=spine_03}}
     {targetBone {boneName=spine_04}}
     {targetBone {boneName=spine_05}}
     {targetBone {boneName=clavicle_r}}
     {targetBone {boneName=pelvis}}
    ]
    constraintWeights [1,1,1,1,1,1,1,1]
   }
   n50 = "Constraint - dyn_chest_dagger_02_base (PPPPPPPP)"(
    ComponentPose=@n49
    Alpha=1
    ConstraintWeights_0=0.1681
    ConstraintWeights_1=0.0351
    ConstraintWeights_2=0.1477
    ConstraintWeights_3=0.2823
    ConstraintWeights_4=0.3184
    ConstraintWeights_5=0.0519
    ConstraintWeights_6=0
    ConstraintWeights_7=0.0001
   ) {
    alphaBoolBlend {blendOption=Linear}
    boneToModify {boneName=dyn_chest_dagger_02_base}
    constraintSetup [
     {targetBone {boneName=clavicle_l}}
     {targetBone {boneName=spine_01}}
     {targetBone {boneName=spine_02}}
     {targetBone {boneName=spine_03}}
     {targetBone {boneName=spine_04}}
     {targetBone {boneName=spine_05}}
     {targetBone {boneName=clavicle_r}}
     {targetBone {boneName=pelvis}}
    ]
    constraintWeights [1,1,1,1,1,1,1,1]
   }
   n51 = Anim Dynamics - Bone: dyn_chest_dagger_00(ComponentPose=@n50, Alpha=0.6, ExternalForce="0,0,0") {
    angularDampingOverride=5
    bGravityOverrideInSimSpace=true
    bUseGravityOverride=true
    alphaBoolBlend {blendOption=Linear}
    boundBone {boneName=dyn_chest_dagger_00}
    gravityOverride {z=-980}
    physicsBodyDefinitions [{
      sphereCollisionRadius=0
      boundBone {boneName=dyn_chest_dagger_00}
      boxExtents=<15 15 15>
      constraintSetup {twistAxis=AxisZ, angularLimitsMax=<1 1 4>, angularLimitsMin=<-1 -1 -4>, angularTarget=<0 0 0>}
      localJointOffset=<0 0 -1>
     }]
    relativeSpaceBone {boneName=root}
   }
   n52 = Anim Dynamics - Bone: dyn_chest_dagger_01(ComponentPose=@n51, Alpha=0.5, ExternalForce="0,0,0") {
    angularDampingOverride=5
    bGravityOverrideInSimSpace=true
    bUseGravityOverride=true
    alphaBoolBlend {blendOption=Linear}
    boundBone {boneName=dyn_chest_dagger_01}
    gravityOverride {z=-980}
    physicsBodyDefinitions [{
      sphereCollisionRadius=0
      boundBone {boneName=dyn_chest_dagger_01}
      boxExtents=<13 13 13>
      constraintSetup {twistAxis=AxisZ, angularLimitsMax=<4 4 0>, angularLimitsMin=<-4 -4 0>, angularTarget=<0 0 0>}
      localJointOffset=<0 0 -1>
     }]
    relativeSpaceBone {boneName=root}
   }
   n53 = Anim Dynamics - Bone: dyn_chest_dagger_02(ComponentPose=@n52, Alpha=0.4, ExternalForce="0,0,0") {
    angularDampingOverride=5
    bGravityOverrideInSimSpace=true
    bUseGravityOverride=true
    alphaBoolBlend {blendOption=Linear}
    boundBone {boneName=dyn_chest_dagger_02}
    gravityOverride {z=-980}
    physicsBodyDefinitions [{
      sphereCollisionRadius=0
      boundBone {boneName=dyn_chest_dagger_02}
      boxExtents=<15 15 15>
      constraintSetup {twistAxis=AxisZ, angularLimitsMax=<4 4 0>, angularLimitsMin=<-4 -4 0>, angularTarget=<0 0 0>}
      localJointOffset=<0 0 -1>
     }]
    relativeSpaceBone {boneName=root}
   }
   n54 = Component To Local(ComponentPose=@n53)
   n55 = Get bIsMoving2D()
   n56 = Anim Dynamics - Bone: dyn_potion(ComponentPose=Local To Component(LocalPose=Copy Pose From Mesh() {bUseAttachedParent=true}), Alpha=0.5, ExternalForce="0,0,0") {
    angularDampingOverride=1
    angularSpringConstant=20000
    bAngularSpring=true
    bOverrideAngularDamping=true
    bUseGravityOverride=true
    gravityScale=0.35
    simulationSpace=RootRelative
    alphaBoolBlend {blendOption=Linear}
    boundBone {boneName=dyn_potion}
    componentLinearAccScale=<1 1 1>
    componentLinearVelScale=<1 1 1>
    gravityOverride {z=-980}
    physicsBodyDefinitions [{
      boundBone {boneName=dyn_potion}
      boxExtents=<15 10 10>
      constraintSetup {angularLimitsMax=<10 5 15>, angularLimitsMin=<-10 -10 -5>, angularTarget=<0 0 0>}
      localJointOffset=<5 0 0>
     }]
    relativeSpaceBone {boneName=spine_05}
   }
   n57 = Anim Dynamics - Bone: dyn_bag_r(ComponentPose=@n56, Alpha=0.6, ExternalForce="0,0,0") {
    angularDampingOverride=1
    angularSpringConstant=10000
    bAngularSpring=true
    bOverrideAngularDamping=true
    bUseGravityOverride=true
    gravityScale=0.35
    simulationSpace=RootRelative
    alphaBoolBlend {blendOption=Linear}
    boundBone {boneName=dyn_bag_r}
    componentLinearAccScale=<1 1 1>
    componentLinearVelScale=<1 1 1>
    gravityOverride {z=-980}
    physicsBodyDefinitions [{
      boundBone {boneName=dyn_bag_r}
      boxExtents=<15 15 15>
      constraintSetup {angularLimitsMax=<5 10 2>, angularLimitsMin=<-5 -15 -2>, angularTarget=<0 0 0>}
      localJointOffset=<3 0 0>
     }]
    relativeSpaceBone {boneName=pelvis}
   }
   n58 = Anim Dynamics - Bone: dyn_bag_l(ComponentPose=@n57, Alpha=0.4, ExternalForce="0,0,0") {
    angularDampingOverride=0.9999
    angularSpringConstant=5000
    bAngularSpring=true
    bOverrideAngularDamping=true
    bUseGravityOverride=true
    gravityScale=0.35
    simulationSpace=RootRelative
    alphaBoolBlend {blendOption=Linear}
    boundBone {boneName=dyn_bag_l}
    componentLinearAccScale=<1 1 1>
    componentLinearVelScale=<1 1 1>
    gravityOverride {z=-980}
    physicsBodyDefinitions [{
      boundBone {boneName=dyn_bag_l}
      boxExtents=<10 10 10>
      constraintSetup {angularLimitsMax=<1 10 1>, angularLimitsMin=<-1 -10 -1>, angularTarget=<0 0 0>}
      localJointOffset=<5 0 0>
     }]
   }
   n59 = Anim Dynamics - Bone: dyn_belt_knife(ComponentPose=@n58, Alpha=0.3, ExternalForce="0,0,0") {
    angularDampingOverride=1
    angularSpringConstant=20000
    bAngularSpring=true
    bOverrideAngularDamping=true
    bUseGravityOverride=true
    gravityScale=0.35
    simulationSpace=RootRelative
    alphaBoolBlend {blendOption=Linear}
    boundBone {boneName=dyn_belt_knife}
    componentLinearAccScale=<1 1 1>
    componentLinearVelScale=<1 1 1>
    gravityOverride {z=-980}
    physicsBodyDefinitions [{
      boundBone {boneName=dyn_belt_knife}
      boxExtents=<10 10 10>
      constraintSetup {angularLimitsMax=<0 1 5>, angularLimitsMin=<0 -10 -5>, angularTarget=<0 0 0>}
      localJointOffset=<5 0 0>
     }]
    relativeSpaceBone {boneName=pelvis}
   }
   n60 = "Transform (Modify) Bone - Bone: dyn_belt_knife"(ComponentPose=@n59, Alpha=1, Translation="0.5,0,0", Rotation="5,0,0", Scale="1,1,1") {
    rotationMode=BMM_Additive
    translationMode=BMM_Additive
    alphaBoolBlend {blendOption=Linear}
    boneToModify {boneName=dyn_belt_knife}
    rotation {pitch=5}
    translation {x=0.5}
   }
   n61 = "Constraint - dyn_skirt_front_01_l (P)"(ComponentPose=@n60, Alpha=0.5, ConstraintWeights_0=1) {
    alphaBoolBlend {blendOption=Linear}
    boneToModify {boneName=dyn_skirt_front_01_l}
    constraintSetup [{targetBone {boneName=thigh_l}}]
    constraintWeights [1]
   }
   n62 = "Constraint - dyn_skirt_front_01_r (P)"(ComponentPose=@n61, Alpha=1, ConstraintWeights_0=1) {
    alphaBoolBlend {blendOption=Linear}
    boneToModify {boneName=dyn_skirt_front_01_r}
    constraintSetup [{targetBone {boneName=thigh_r}}]
    constraintWeights [1]
   }
   n63 = RigidBody(ComponentPose=@n62, ExternalForce=Get GravityOverride()) {
    bClampLinearTranslationLimitToRefPose=true
    overridePhysicsAsset=&M_MED_Swamp_Knight_Skirt_Physics
    simulationSpace=BaseBoneSpace
    alphaBoolBlend {blendOption=Linear}
    baseBoneRef {boneName=pelvis}
    componentAppliedLinearAccClamp=<2750 2750 2750>
    componentLinearAccScale=<0.25 0.25 0.15>
    componentLinearVelScale=<0.25 0.25 0.15>
    overrideWorldGravity {z=-980}
   }
   n64 = "Constraint - dyn_skirt_front_01_r (P)"(ComponentPose=@n63, Alpha=1, ConstraintWeights_0=1) {
    alphaBoolBlend {blendOption=Linear}
    boneToModify {boneName=dyn_skirt_front_01_r}
    constraintSetup [{targetBone {boneName=thigh_r}}]
    constraintWeights [1]
   }
   n65 = "Transform (Modify) Bone - Bone: dyn_bag"(ComponentPose=@n64, Alpha=1, Translation="0,2,4", Rotation="0,0,0", Scale="1,1,1") {
    rotationMode=BMM_Additive
    translationMode=BMM_Additive
    alphaBoolBlend {blendOption=Linear}
    boneToModify {boneName=dyn_bag}
    translation {y=2, z=4}
   }
   n66 = "Constraint - dyn_chest_dagger_00_base (PPPPPPPP)"(
    ComponentPose=@n65
    Alpha=1
    ConstraintWeights_0=0.006
    ConstraintWeights_1=0.0476
    ConstraintWeights_2=0.187
    ConstraintWeights_3=0.3862
    ConstraintWeights_4=0.2588
    ConstraintWeights_5=0.0289
    ConstraintWeights_6=0.0831
    ConstraintWeights_7=0.0022
   ) {
    alphaBoolBlend {blendOption=Linear}
    boneToModify {boneName=dyn_chest_dagger_00_base}
    constraintSetup [
     {targetBone {boneName=clavicle_l}}
     {targetBone {boneName=spine_01}}
     {targetBone {boneName=spine_02}}
     {targetBone {boneName=spine_03}}
     {targetBone {boneName=spine_04}}
     {targetBone {boneName=spine_05}}
     {targetBone {boneName=clavicle_r}}
     {targetBone {boneName=pelvis}}
    ]
    constraintWeights [1,1,1,1,1,1,1,1]
   }
   n67 = "Constraint - dyn_chest_dagger_01_base (PPPPPPPP)"(
    ComponentPose=@n66
    Alpha=1
    ConstraintWeights_0=0.1183
    ConstraintWeights_1=0.0377
    ConstraintWeights_2=0.1683
    ConstraintWeights_3=0.3301
    ConstraintWeights_4=0.2998
    ConstraintWeights_5=0.0458
    ConstraintWeights_6=0
    ConstraintWeights_7=0
   ) {
    alphaBoolBlend {blendOption=Linear}
    boneToModify {boneName=dyn_chest_dagger_01_base}
    constraintSetup [
     {targetBone {boneName=clavicle_l}}
     {targetBone {boneName=spine_01}}
     {targetBone {boneName=spine_02}}
     {targetBone {boneName=spine_03}}
     {targetBone {boneName=spine_04}}
     {targetBone {boneName=spine_05}}
     {targetBone {boneName=clavicle_r}}
     {targetBone {boneName=pelvis}}
    ]
    constraintWeights [1,1,1,1,1,1,1,1]
   }
   n68 = "Constraint - dyn_chest_dagger_02_base (PPPPPPPP)"(
    ComponentPose=@n67
    Alpha=1
    ConstraintWeights_0=0.1681
    ConstraintWeights_1=0.0351
    ConstraintWeights_2=0.1477
    ConstraintWeights_3=0.2823
    ConstraintWeights_4=0.3184
    ConstraintWeights_5=0.0519
    ConstraintWeights_6=0
    ConstraintWeights_7=0.0001
   ) {
    alphaBoolBlend {blendOption=Linear}
    boneToModify {boneName=dyn_chest_dagger_02_base}
    constraintSetup [
     {targetBone {boneName=clavicle_l}}
     {targetBone {boneName=spine_01}}
     {targetBone {boneName=spine_02}}
     {targetBone {boneName=spine_03}}
     {targetBone {boneName=spine_04}}
     {targetBone {boneName=spine_05}}
     {targetBone {boneName=clavicle_r}}
     {targetBone {boneName=pelvis}}
    ]
    constraintWeights [1,1,1,1,1,1,1,1]
   }
   n69 = Anim Dynamics - Bone: dyn_chest_dagger_00(ComponentPose=@n68, Alpha=0.6, ExternalForce="0,0,0") {
    angularDampingOverride=5
    bGravityOverrideInSimSpace=true
    bUseGravityOverride=true
    alphaBoolBlend {blendOption=Linear}
    boundBone {boneName=dyn_chest_dagger_00}
    gravityOverride {z=-980}
    physicsBodyDefinitions [{
      sphereCollisionRadius=0
      boundBone {boneName=dyn_chest_dagger_00}
      boxExtents=<15 15 15>
      constraintSetup {twistAxis=AxisZ, angularLimitsMax=<1 1 4>, angularLimitsMin=<-1 -1 -4>, angularTarget=<0 0 0>}
      localJointOffset=<0 0 -1>
     }]
    relativeSpaceBone {boneName=root}
   }
   n70 = Anim Dynamics - Bone: dyn_chest_dagger_01(ComponentPose=@n69, Alpha=0.5, ExternalForce="0,0,0") {
    angularDampingOverride=5
    bGravityOverrideInSimSpace=true
    bUseGravityOverride=true
    alphaBoolBlend {blendOption=Linear}
    boundBone {boneName=dyn_chest_dagger_01}
    gravityOverride {z=-980}
    physicsBodyDefinitions [{
      sphereCollisionRadius=0
      boundBone {boneName=dyn_chest_dagger_01}
      boxExtents=<13 13 13>
      constraintSetup {twistAxis=AxisZ, angularLimitsMax=<4 4 0>, angularLimitsMin=<-4 -4 0>, angularTarget=<0 0 0>}
      localJointOffset=<0 0 -1>
     }]
    relativeSpaceBone {boneName=root}
   }
   n71 = Anim Dynamics - Bone: dyn_chest_dagger_02(ComponentPose=@n70, Alpha=0.4, ExternalForce="0,0,0") {
    angularDampingOverride=5
    bGravityOverrideInSimSpace=true
    bUseGravityOverride=true
    alphaBoolBlend {blendOption=Linear}
    boundBone {boneName=dyn_chest_dagger_02}
    gravityOverride {z=-980}
    physicsBodyDefinitions [{
      sphereCollisionRadius=0
      boundBone {boneName=dyn_chest_dagger_02}
      boxExtents=<15 15 15>
      constraintSetup {twistAxis=AxisZ, angularLimitsMax=<4 4 0>, angularLimitsMin=<-4 -4 0>, angularTarget=<0 0 0>}
      localJointOffset=<0 0 -1>
     }]
    relativeSpaceBone {boneName=root}
   }
   n72 = Component To Local(ComponentPose=@n71)
   n73 = Anim Dynamics - Bone: dyn_potion(ComponentPose=Local To Component(LocalPose=Copy Pose From Mesh() {bUseAttachedParent=true}), Alpha=0.5, ExternalForce="0,0,0") {
    angularDampingOverride=1
    angularSpringConstant=20000
    bAngularSpring=true
    bOverrideAngularDamping=true
    bUseGravityOverride=true
    gravityScale=0.35
    simulationSpace=RootRelative
    alphaBoolBlend {blendOption=Linear}
    boundBone {boneName=dyn_potion}
    componentLinearAccScale=<1 1 1>
    componentLinearVelScale=<1 1 1>
    gravityOverride {z=-980}
    physicsBodyDefinitions [{
      boundBone {boneName=dyn_potion}
      boxExtents=<15 10 10>
      constraintSetup {angularLimitsMax=<10 5 15>, angularLimitsMin=<-10 -10 -5>, angularTarget=<0 0 0>}
      localJointOffset=<5 0 0>
     }]
    relativeSpaceBone {boneName=spine_05}
   }
   n74 = Anim Dynamics - Bone: dyn_bag_r(ComponentPose=@n73, Alpha=0.5, ExternalForce="0,0,0") {
    angularDampingOverride=1
    angularSpringConstant=10000
    bAngularSpring=true
    bOverrideAngularDamping=true
    bUseGravityOverride=true
    gravityScale=0.35
    simulationSpace=RootRelative
    alphaBoolBlend {blendOption=Linear}
    boundBone {boneName=dyn_bag_r}
    componentLinearAccScale=<1 1 1>
    componentLinearVelScale=<1 1 1>
    gravityOverride {z=-980}
    physicsBodyDefinitions [{
      boundBone {boneName=dyn_bag_r}
      boxExtents=<15 15 15>
      constraintSetup {angularLimitsMax=<5 10 2>, angularLimitsMin=<-5 -15 -2>, angularTarget=<0 0 0>}
      localJointOffset=<3 0 0>
     }]
    relativeSpaceBone {boneName=pelvis}
   }
   n75 = Anim Dynamics - Bone: dyn_bag_l(ComponentPose=@n74, Alpha=0.4, ExternalForce="0,0,0") {
    angularDampingOverride=0.9999
    angularSpringConstant=5000
    bAngularSpring=true
    bOverrideAngularDamping=true
    bUseGravityOverride=true
    gravityScale=0.35
    simulationSpace=RootRelative
    alphaBoolBlend {blendOption=Linear}
    boundBone {boneName=dyn_bag_l}
    componentLinearAccScale=<1 1 1>
    componentLinearVelScale=<1 1 1>
    gravityOverride {z=-980}
    physicsBodyDefinitions [{
      boundBone {boneName=dyn_bag_l}
      boxExtents=<10 10 10>
      constraintSetup {angularLimitsMax=<1 10 1>, angularLimitsMin=<-1 -10 -1>, angularTarget=<0 0 0>}
      localJointOffset=<5 0 0>
     }]
   }
   n76 = Anim Dynamics - Bone: dyn_belt_knife(ComponentPose=@n75, Alpha=0.3, ExternalForce="0,0,0") {
    angularDampingOverride=1
    angularSpringConstant=20000
    bAngularSpring=true
    bOverrideAngularDamping=true
    bUseGravityOverride=true
    gravityScale=0.35
    simulationSpace=RootRelative
    alphaBoolBlend {blendOption=Linear}
    boundBone {boneName=dyn_belt_knife}
    componentLinearAccScale=<1 1 1>
    componentLinearVelScale=<1 1 1>
    gravityOverride {z=-980}
    physicsBodyDefinitions [{
      boundBone {boneName=dyn_belt_knife}
      boxExtents=<10 10 10>
      constraintSetup {angularLimitsMax=<0 1 5>, angularLimitsMin=<0 -10 -5>, angularTarget=<0 0 0>}
      localJointOffset=<5 0 0>
     }]
    relativeSpaceBone {boneName=pelvis}
   }
   n77 = "Transform (Modify) Bone - Bone: dyn_belt_knife"(ComponentPose=@n76, Alpha=1, Translation="0.5,0,0", Rotation="5,0,0", Scale="1,1,1") {
    rotationMode=BMM_Additive
    translationMode=BMM_Additive
    alphaBoolBlend {blendOption=Linear}
    boneToModify {boneName=dyn_belt_knife}
    rotation {pitch=5}
    translation {x=0.5}
   }
   n78 = "Constraint - dyn_skirt_front_01_l (P)"(ComponentPose=@n77, Alpha=0.8, ConstraintWeights_0=1) {
    alphaBoolBlend {blendOption=Linear}
    boneToModify {boneName=dyn_skirt_front_01_l}
    constraintSetup [{targetBone {boneName=thigh_l}}]
    constraintWeights [1]
   }
   n79 = "Constraint - dyn_skirt_front_01_r (P)"(ComponentPose=@n78, Alpha=0.6, ConstraintWeights_0=1) {
    alphaBoolBlend {blendOption=Linear}
    boneToModify {boneName=dyn_skirt_front_01_r}
    constraintSetup [{targetBone {boneName=thigh_r}}]
    constraintWeights [1]
   }
   n80 = RigidBody(ComponentPose=@n79, ExternalForce=Get GravityOverride()) {
    bClampLinearTranslationLimitToRefPose=true
    overridePhysicsAsset=&M_MED_Swamp_Knight_Skirt_Physics
    simulationSpace=BaseBoneSpace
    alphaBoolBlend {blendOption=Linear}
    baseBoneRef {boneName=pelvis}
    componentAppliedLinearAccClamp=<2750 2750 2750>
    componentLinearAccScale=<0.25 0.25 0.15>
    componentLinearVelScale=<0.25 0.25 0.15>
    overrideWorldGravity {z=-980}
   }
   n81 = "Constraint - dyn_skirt_front_01_r (P)"(ComponentPose=@n80, Alpha=0.3, ConstraintWeights_0=1) {
    alphaBoolBlend {blendOption=Linear}
    boneToModify {boneName=dyn_skirt_front_01_r}
    constraintSetup [{targetBone {boneName=thigh_r}}]
    constraintWeights [1]
   }
   n82 = "Constraint - dyn_skirt_front_02_r (PPPPPP)"(
    ComponentPose=@n81
    Alpha=0.4
    ConstraintWeights_0=0.0287
    ConstraintWeights_1=0.0225
    ConstraintWeights_2=0.0197
    ConstraintWeights_3=0.6174
    ConstraintWeights_4=0.04
    ConstraintWeights_5=0.2717
   ) {
    alphaBoolBlend {blendOption=Linear}
    boneToModify {boneName=dyn_skirt_front_02_r}
    constraintSetup [
     {targetBone {boneName=spine_01}}
     {targetBone {boneName=dyn_skirt_front_01_r}}
     {targetBone {boneName=thigh_r}}
     {targetBone {boneName=thigh_twist_01_r}}
     {targetBone {boneName=thigh_twist_01_l}}
     {targetBone {boneName=pelvis}}
    ]
    constraintWeights [1,1,1,1,1,1]
   }
   n83 = "Constraint - dyn_skirt_front_03_r (PPPP)"(
    ComponentPose=@n82
    Alpha=0.4
    ConstraintWeights_0=0.0054
    ConstraintWeights_1=0.1072
    ConstraintWeights_2=0.8096
    ConstraintWeights_3=0.0777
   ) {
    alphaBoolBlend {blendOption=Linear}
    boneToModify {boneName=dyn_skirt_front_03_r}
    constraintSetup [{targetBone {boneName=spine_01}}, {targetBone {boneName=thigh_r}}, {targetBone {boneName=thigh_twist_01_r}}, {targetBone {boneName=pelvis}}]
    constraintWeights [1,1,1,1]
   }
   n84 = "Constraint - dyn_skirt_front_04_l (PPPP)"!(
    ComponentPose=@n83
    Alpha=0.5
    ConstraintWeights_0=0.4361
    ConstraintWeights_1:float=0.0615
    ConstraintWeights_2:float=0.0026
    ConstraintWeights_3:float=0.4999
   ) -> (Pose) {
    alphaBoolBlend {blendOption=Linear}
    boneToModify {boneName=dyn_skirt_front_04_l}
    constraintSetup [
     {targetBone {boneName=thigh_l}}
     {targetBone {boneName=calf_twist_02_l}}
     {targetBone {boneName=thigh_twist_01_l}}
     {targetBone {boneName=calf_l}}
    ]
    constraintWeights [1,1,1,1]
   }
   n85 = "Constraint - dyn_chest_dagger_00_base (PPPPPPPP)"(
    ComponentPose=@n84
    Alpha=1
    ConstraintWeights_0=0.006
    ConstraintWeights_1=0.0476
    ConstraintWeights_2=0.187
    ConstraintWeights_3=0.3862
    ConstraintWeights_4=0.2588
    ConstraintWeights_5=0.0289
    ConstraintWeights_6=0.0831
    ConstraintWeights_7=0.0022
   ) {
    alphaBoolBlend {blendOption=Linear}
    boneToModify {boneName=dyn_chest_dagger_00_base}
    constraintSetup [
     {targetBone {boneName=clavicle_l}}
     {targetBone {boneName=spine_01}}
     {targetBone {boneName=spine_02}}
     {targetBone {boneName=spine_03}}
     {targetBone {boneName=spine_04}}
     {targetBone {boneName=spine_05}}
     {targetBone {boneName=clavicle_r}}
     {targetBone {boneName=pelvis}}
    ]
    constraintWeights [1,1,1,1,1,1,1,1]
   }
   n86 = "Constraint - dyn_chest_dagger_01_base (PPPPPPPP)"(
    ComponentPose=@n85
    Alpha=1
    ConstraintWeights_0=0.1183
    ConstraintWeights_1=0.0377
    ConstraintWeights_2=0.1683
    ConstraintWeights_3=0.3301
    ConstraintWeights_4=0.2998
    ConstraintWeights_5=0.0458
    ConstraintWeights_6=0
    ConstraintWeights_7=0
   ) {
    alphaBoolBlend {blendOption=Linear}
    boneToModify {boneName=dyn_chest_dagger_01_base}
    constraintSetup [
     {targetBone {boneName=clavicle_l}}
     {targetBone {boneName=spine_01}}
     {targetBone {boneName=spine_02}}
     {targetBone {boneName=spine_03}}
     {targetBone {boneName=spine_04}}
     {targetBone {boneName=spine_05}}
     {targetBone {boneName=clavicle_r}}
     {targetBone {boneName=pelvis}}
    ]
    constraintWeights [1,1,1,1,1,1,1,1]
   }
   n87 = "Constraint - dyn_chest_dagger_02_base (PPPPPPPP)"(
    ComponentPose=@n86
    Alpha=1
    ConstraintWeights_0=0.1681
    ConstraintWeights_1=0.0351
    ConstraintWeights_2=0.1477
    ConstraintWeights_3=0.2823
    ConstraintWeights_4=0.3184
    ConstraintWeights_5=0.0519
    ConstraintWeights_6=0
    ConstraintWeights_7=0.0001
   ) {
    alphaBoolBlend {blendOption=Linear}
    boneToModify {boneName=dyn_chest_dagger_02_base}
    constraintSetup [
     {targetBone {boneName=clavicle_l}}
     {targetBone {boneName=spine_01}}
     {targetBone {boneName=spine_02}}
     {targetBone {boneName=spine_03}}
     {targetBone {boneName=spine_04}}
     {targetBone {boneName=spine_05}}
     {targetBone {boneName=clavicle_r}}
     {targetBone {boneName=pelvis}}
    ]
    constraintWeights [1,1,1,1,1,1,1,1]
   }
   n88 = Anim Dynamics - Bone: dyn_chest_dagger_00(ComponentPose=@n87, Alpha=0.6, ExternalForce="0,0,0") {
    angularDampingOverride=5
    bGravityOverrideInSimSpace=true
    bUseGravityOverride=true
    alphaBoolBlend {blendOption=Linear}
    boundBone {boneName=dyn_chest_dagger_00}
    gravityOverride {z=-980}
    physicsBodyDefinitions [{
      sphereCollisionRadius=0
      boundBone {boneName=dyn_chest_dagger_00}
      boxExtents=<15 15 15>
      constraintSetup {twistAxis=AxisZ, angularLimitsMax=<1 1 4>, angularLimitsMin=<-1 -1 -4>, angularTarget=<0 0 0>}
      localJointOffset=<0 0 -1>
     }]
    relativeSpaceBone {boneName=root}
   }
   n89 = Anim Dynamics - Bone: dyn_chest_dagger_01(ComponentPose=@n88, Alpha=0.5, ExternalForce="0,0,0") {
    angularDampingOverride=5
    bGravityOverrideInSimSpace=true
    bUseGravityOverride=true
    alphaBoolBlend {blendOption=Linear}
    boundBone {boneName=dyn_chest_dagger_01}
    gravityOverride {z=-980}
    physicsBodyDefinitions [{
      sphereCollisionRadius=0
      boundBone {boneName=dyn_chest_dagger_01}
      boxExtents=<13 13 13>
      constraintSetup {twistAxis=AxisZ, angularLimitsMax=<4 4 0>, angularLimitsMin=<-4 -4 0>, angularTarget=<0 0 0>}
      localJointOffset=<0 0 -1>
     }]
    relativeSpaceBone {boneName=root}
   }
   n90 = Anim Dynamics - Bone: dyn_chest_dagger_02(ComponentPose=@n89, Alpha=0.4, ExternalForce="0,0,0") {
    angularDampingOverride=5
    bGravityOverrideInSimSpace=true
    bUseGravityOverride=true
    alphaBoolBlend {blendOption=Linear}
    boundBone {boneName=dyn_chest_dagger_02}
    gravityOverride {z=-980}
    physicsBodyDefinitions [{
      sphereCollisionRadius=0
      boundBone {boneName=dyn_chest_dagger_02}
      boxExtents=<15 15 15>
      constraintSetup {twistAxis=AxisZ, angularLimitsMax=<4 4 0>, angularLimitsMin=<-4 -4 0>, angularTarget=<0 0 0>}
      localJointOffset=<0 0 -1>
     }]
    relativeSpaceBone {boneName=root}
   }
   n91 = Component To Local(ComponentPose=@n90)
   n92 = Get bIsJumping()
   n93 = RigidBody(ComponentPose=Local To Component(LocalPose=Copy Pose From Mesh() {bUseAttachedParent=true}), ExternalForce=Get GravityOverride()) {
    overridePhysicsAsset=&/M_MED_Swamp_Knight_Skydive_Physics
    alphaBoolBlend {blendOption=Linear}
    componentAppliedLinearAccClamp=<5000 5000 5000>
    componentLinearAccScale=<0.3 0.3 0.4>
    componentLinearVelScale=<0.3 0.3 0.4>
   }
   n94 = "Constraint - dyn_skirt_front_02_r (PPPPPP)"(
    ComponentPose=@n93
    Alpha=0.4
    ConstraintWeights_0=0.0287
    ConstraintWeights_1=0.0225
    ConstraintWeights_2=0.0197
    ConstraintWeights_3=0.6174
    ConstraintWeights_4=0.04
    ConstraintWeights_5=0.2717
   ) {
    alphaBoolBlend {blendOption=Linear}
    boneToModify {boneName=dyn_skirt_front_02_r}
    constraintSetup [
     {targetBone {boneName=spine_01}}
     {targetBone {boneName=dyn_skirt_front_01_r}}
     {targetBone {boneName=thigh_r}}
     {targetBone {boneName=thigh_twist_01_r}}
     {targetBone {boneName=thigh_twist_01_l}}
     {targetBone {boneName=pelvis}}
    ]
    constraintWeights [1,1,1,1,1,1]
   }
   n95 = "Constraint - dyn_skirt_front_03_r (PPPP)"(
    ComponentPose=@n94
    Alpha=0.4
    ConstraintWeights_0=0.0054
    ConstraintWeights_1=0.1072
    ConstraintWeights_2=0.8096
    ConstraintWeights_3=0.0777
   ) {
    alphaBoolBlend {blendOption=Linear}
    boneToModify {boneName=dyn_skirt_front_03_r}
    constraintSetup [{targetBone {boneName=spine_01}}, {targetBone {boneName=thigh_r}}, {targetBone {boneName=thigh_twist_01_r}}, {targetBone {boneName=pelvis}}]
    constraintWeights [1,1,1,1]
   }
   n96 = "Constraint - dyn_skirt_front_04_l (PPPP)"(
    ComponentPose=@n95
    Alpha=0.5
    ConstraintWeights_0=0.4361
    ConstraintWeights_1=0.0615
    ConstraintWeights_2=0.0026
    ConstraintWeights_3=0.4999
   ) {
    alphaBoolBlend {blendOption=Linear}
    boneToModify {boneName=dyn_skirt_front_04_l}
    constraintSetup [
     {targetBone {boneName=thigh_l}}
     {targetBone {boneName=calf_twist_02_l}}
     {targetBone {boneName=thigh_twist_01_l}}
     {targetBone {boneName=calf_l}}
    ]
    constraintWeights [1,1,1,1]
   }
   n97 = Blend Poses by bool(
    BlendPose_0=Component To Local(ComponentPose=@n96)
    BlendPose_1=Blend Poses by bool(
     BlendPose_0=Blend Poses by bool(BlendPose_0=@n25, BlendPose_1=@n54, BlendTime_0=0.3, BlendTime_1=0.2, bActiveValue=@n55) {blendType=Linear}
     BlendPose_1=Blend Poses by bool(BlendPose_0=@n72, BlendPose_1=@n91, BlendTime_0=0.3, BlendTime_1=0.5, bActiveValue=@n92) {blendType=Linear}
     BlendTime_0=0.2
     BlendTime_1=0.2
     bActiveValue=Get bIsCrouching()
    ) {blendType=Linear}
    BlendTime_0=0.3
    BlendTime_1=0.5
    bActiveValue=Get DiveOrFloat() -> (DiveOrFloat:bool)
   ) {blendType=Linear}
   n98 = Anim Dynamics - Bone: dyn_potion(ComponentPose=Local To Component(LocalPose=Copy Pose From Mesh() {bUseAttachedParent=true}), Alpha=0.5, ExternalForce="0,0,0") {
    angularDampingOverride=1
    angularSpringConstant=20000
    bAngularSpring=true
    bOverrideAngularDamping=true
    bUseGravityOverride=true
    gravityScale=0.35
    simulationSpace=RootRelative
    alphaBoolBlend {blendOption=Linear}
    boundBone {boneName=dyn_potion}
    componentLinearAccScale=<1 1 1>
    componentLinearVelScale=<1 1 1>
    gravityOverride {z=-980}
    physicsBodyDefinitions [{
      boundBone {boneName=dyn_potion}
      boxExtents=<15 10 10>
      constraintSetup {angularLimitsMax=<10 5 15>, angularLimitsMin=<-10 -10 -5>, angularTarget=<0 0 0>}
      localJointOffset=<5 0 0>
     }]
    relativeSpaceBone {boneName=spine_05}
   }
   n99 = Anim Dynamics - Bone: dyn_bag_r(ComponentPose=@n98, Alpha=0.5, ExternalForce="0,0,0") {
    angularDampingOverride=1
    angularSpringConstant=10000
    bAngularSpring=true
    bOverrideAngularDamping=true
    bUseGravityOverride=true
    gravityScale=0.35
    simulationSpace=RootRelative
    alphaBoolBlend {blendOption=Linear}
    boundBone {boneName=dyn_bag_r}
    componentLinearAccScale=<1 1 1>
    componentLinearVelScale=<1 1 1>
    gravityOverride {z=-980}
    physicsBodyDefinitions [{
      boundBone {boneName=dyn_bag_r}
      boxExtents=<15 15 15>
      constraintSetup {angularLimitsMax=<5 10 2>, angularLimitsMin=<-5 -15 -2>, angularTarget=<0 0 0>}
      localJointOffset=<3 0 0>
     }]
    relativeSpaceBone {boneName=pelvis}
   }
   n100 = Anim Dynamics - Bone: dyn_bag_l(ComponentPose=@n99, Alpha=0.4, ExternalForce="0,0,0") {
    angularDampingOverride=0.9999
    angularSpringConstant=5000
    bAngularSpring=true
    bOverrideAngularDamping=true
    bUseGravityOverride=true
    gravityScale=0.35
    simulationSpace=RootRelative
    alphaBoolBlend {blendOption=Linear}
    boundBone {boneName=dyn_bag_l}
    componentLinearAccScale=<1 1 1>
    componentLinearVelScale=<1 1 1>
    gravityOverride {z=-980}
    physicsBodyDefinitions [{
      boundBone {boneName=dyn_bag_l}
      boxExtents=<10 10 10>
      constraintSetup {angularLimitsMax=<1 10 1>, angularLimitsMin=<-1 -10 -1>, angularTarget=<0 0 0>}
      localJointOffset=<5 0 0>
     }]
   }
   n101 = Anim Dynamics - Bone: dyn_belt_knife(ComponentPose=@n100, Alpha=0.3, ExternalForce="0,0,0") {
    angularDampingOverride=1
    angularSpringConstant=20000
    bAngularSpring=true
    bOverrideAngularDamping=true
    bUseGravityOverride=true
    gravityScale=0.35
    simulationSpace=RootRelative
    alphaBoolBlend {blendOption=Linear}
    boundBone {boneName=dyn_belt_knife}
    componentLinearAccScale=<1 1 1>
    componentLinearVelScale=<1 1 1>
    gravityOverride {z=-980}
    physicsBodyDefinitions [{
      boundBone {boneName=dyn_belt_knife}
      boxExtents=<10 10 10>
      constraintSetup {angularLimitsMax=<0 1 5>, angularLimitsMin=<0 -10 -5>, angularTarget=<0 0 0>}
      localJointOffset=<5 0 0>
     }]
    relativeSpaceBone {boneName=pelvis}
   }
   n102 = "Constraint - dyn_skirt_front_01_l (P)"(ComponentPose=@n101, Alpha=0.2, ConstraintWeights_0=1) {
    alphaBoolBlend {blendOption=Linear}
    boneToModify {boneName=dyn_skirt_front_01_l}
    constraintSetup [{targetBone {boneName=thigh_l}}]
    constraintWeights [1]
   }
   n103 = "Constraint - dyn_skirt_front_01_r (P)"(ComponentPose=@n102, Alpha=0.6, ConstraintWeights_0=1) {
    alphaBoolBlend {blendOption=Linear}
    boneToModify {boneName=dyn_skirt_front_01_r}
    constraintSetup [{targetBone {boneName=thigh_r}}]
    constraintWeights [1]
   }
   n104 = RigidBody(ComponentPose=@n103, ExternalForce=Get GravityOverride()) {
    bClampLinearTranslationLimitToRefPose=true
    overridePhysicsAsset=&M_MED_Swamp_Knight_Skirt_Physics
    simulationSpace=BaseBoneSpace
    alphaBoolBlend {blendOption=Linear}
    baseBoneRef {boneName=pelvis}
    componentAppliedLinearAccClamp=<2750 2750 2750>
    componentLinearAccScale=<0.25 0.25 0.15>
    componentLinearVelScale=<0.25 0.25 0.15>
    overrideWorldGravity {z=-980}
   }
   n105 = "Constraint - dyn_skirt_front_01_r (P)"(ComponentPose=@n104, Alpha=0.3, ConstraintWeights_0=1) {
    alphaBoolBlend {blendOption=Linear}
    boneToModify {boneName=dyn_skirt_front_01_r}
    constraintSetup [{targetBone {boneName=thigh_r}}]
    constraintWeights [1]
   }
   n106 = "Transform (Modify) Bone - Bone: dyn_skirt_front_02_l"(ComponentPose=@n105, Alpha=1, Translation="1,1,0", Rotation="0,0,0", Scale="1,1,1") {translationMode=BMM_Additive, alphaBoolBlend {blendOption=Linear}, boneToModify {boneName=dyn_skirt_front_02_l}, translation=<1 1>}
   n107 = "Constraint - dyn_skirt_front_02_r (PPPPPP)"(
    ComponentPose=@n106
    Alpha=0.4
    ConstraintWeights_0=0.0287
    ConstraintWeights_1=0.0225
    ConstraintWeights_2=0.0197
    ConstraintWeights_3=0.6174
    ConstraintWeights_4=0.04
    ConstraintWeights_5=0.2717
   ) {
    alphaBoolBlend {blendOption=Linear}
    boneToModify {boneName=dyn_skirt_front_02_r}
    constraintSetup [
     {targetBone {boneName=spine_01}}
     {targetBone {boneName=dyn_skirt_front_01_r}}
     {targetBone {boneName=thigh_r}}
     {targetBone {boneName=thigh_twist_01_r}}
     {targetBone {boneName=thigh_twist_01_l}}
     {targetBone {boneName=pelvis}}
    ]
    constraintWeights [1,1,1,1,1,1]
   }
   n108 = "Constraint - dyn_skirt_front_03_r (PPPP)"(
    ComponentPose=@n107
    Alpha=0.4
    ConstraintWeights_0=0.0054
    ConstraintWeights_1=0.1072
    ConstraintWeights_2=0.8096
    ConstraintWeights_3=0.0777
   ) {
    alphaBoolBlend {blendOption=Linear}
    boneToModify {boneName=dyn_skirt_front_03_r}
    constraintSetup [{targetBone {boneName=spine_01}}, {targetBone {boneName=thigh_r}}, {targetBone {boneName=thigh_twist_01_r}}, {targetBone {boneName=pelvis}}]
    constraintWeights [1,1,1,1]
   }
   n109 = "Constraint - dyn_skirt_front_04_l (PPPP)"(
    ComponentPose=@n108
    Alpha=0.5
    ConstraintWeights_0=0.4361
    ConstraintWeights_1=0.0615
    ConstraintWeights_2=0.0026
    ConstraintWeights_3=0.4999
   ) {
    alphaBoolBlend {blendOption=Linear}
    boneToModify {boneName=dyn_skirt_front_04_l}
    constraintSetup [
     {targetBone {boneName=thigh_l}}
     {targetBone {boneName=calf_twist_02_l}}
     {targetBone {boneName=thigh_twist_01_l}}
     {targetBone {boneName=calf_l}}
    ]
    constraintWeights [1,1,1,1]
   }
   n110 = "Constraint - dyn_chest_dagger_00_base (PPPPPPPP)"(
    ComponentPose=@n109
    Alpha=1
    ConstraintWeights_0=0.006
    ConstraintWeights_1=0.0476
    ConstraintWeights_2=0.187
    ConstraintWeights_3=0.3862
    ConstraintWeights_4=0.2588
    ConstraintWeights_5=0.0289
    ConstraintWeights_6=0.0831
    ConstraintWeights_7=0.0022
   ) {
    alphaBoolBlend {blendOption=Linear}
    boneToModify {boneName=dyn_chest_dagger_00_base}
    constraintSetup [
     {targetBone {boneName=clavicle_l}}
     {targetBone {boneName=spine_01}}
     {targetBone {boneName=spine_02}}
     {targetBone {boneName=spine_03}}
     {targetBone {boneName=spine_04}}
     {targetBone {boneName=spine_05}}
     {targetBone {boneName=clavicle_r}}
     {targetBone {boneName=pelvis}}
    ]
    constraintWeights [1,1,1,1,1,1,1,1]
   }
   n111 = "Constraint - dyn_chest_dagger_01_base (PPPPPPPP)"(
    ComponentPose=@n110
    Alpha=1
    ConstraintWeights_0=0.1183
    ConstraintWeights_1=0.0377
    ConstraintWeights_2=0.1683
    ConstraintWeights_3=0.3301
    ConstraintWeights_4=0.2998
    ConstraintWeights_5=0.0458
    ConstraintWeights_6=0
    ConstraintWeights_7=0
   ) {
    alphaBoolBlend {blendOption=Linear}
    boneToModify {boneName=dyn_chest_dagger_01_base}
    constraintSetup [
     {targetBone {boneName=clavicle_l}}
     {targetBone {boneName=spine_01}}
     {targetBone {boneName=spine_02}}
     {targetBone {boneName=spine_03}}
     {targetBone {boneName=spine_04}}
     {targetBone {boneName=spine_05}}
     {targetBone {boneName=clavicle_r}}
     {targetBone {boneName=pelvis}}
    ]
    constraintWeights [1,1,1,1,1,1,1,1]
   }
   n112 = "Constraint - dyn_chest_dagger_02_base (PPPPPPPP)"(
    ComponentPose=@n111
    Alpha=1
    ConstraintWeights_0=0.1681
    ConstraintWeights_1=0.0351
    ConstraintWeights_2=0.1477
    ConstraintWeights_3=0.2823
    ConstraintWeights_4=0.3184
    ConstraintWeights_5=0.0519
    ConstraintWeights_6=0
    ConstraintWeights_7=0.0001
   ) {
    alphaBoolBlend {blendOption=Linear}
    boneToModify {boneName=dyn_chest_dagger_02_base}
    constraintSetup [
     {targetBone {boneName=clavicle_l}}
     {targetBone {boneName=spine_01}}
     {targetBone {boneName=spine_02}}
     {targetBone {boneName=spine_03}}
     {targetBone {boneName=spine_04}}
     {targetBone {boneName=spine_05}}
     {targetBone {boneName=clavicle_r}}
     {targetBone {boneName=pelvis}}
    ]
    constraintWeights [1,1,1,1,1,1,1,1]
   }
   n113 = Anim Dynamics - Bone: dyn_chest_dagger_00(ComponentPose=@n112, Alpha=0.6, ExternalForce="0,0,0") {
    angularDampingOverride=5
    bGravityOverrideInSimSpace=true
    bUseGravityOverride=true
    alphaBoolBlend {blendOption=Linear}
    boundBone {boneName=dyn_chest_dagger_00}
    gravityOverride {z=-980}
    physicsBodyDefinitions [{
      sphereCollisionRadius=0
      boundBone {boneName=dyn_chest_dagger_00}
      boxExtents=<15 15 15>
      constraintSetup {twistAxis=AxisZ, angularLimitsMax=<1 1 4>, angularLimitsMin=<-1 -1 -4>, angularTarget=<0 0 0>}
      localJointOffset=<0 0 -1>
     }]
    relativeSpaceBone {boneName=root}
   }
   n114 = Anim Dynamics - Bone: dyn_chest_dagger_01(ComponentPose=@n113, Alpha=0.5, ExternalForce="0,0,0") {
    angularDampingOverride=5
    bGravityOverrideInSimSpace=true
    bUseGravityOverride=true
    alphaBoolBlend {blendOption=Linear}
    boundBone {boneName=dyn_chest_dagger_01}
    gravityOverride {z=-980}
    physicsBodyDefinitions [{
      sphereCollisionRadius=0
      boundBone {boneName=dyn_chest_dagger_01}
      boxExtents=<13 13 13>
      constraintSetup {twistAxis=AxisZ, angularLimitsMax=<4 4 0>, angularLimitsMin=<-4 -4 0>, angularTarget=<0 0 0>}
      localJointOffset=<0 0 -1>
     }]
    relativeSpaceBone {boneName=root}
   }
   n115 = Anim Dynamics - Bone: dyn_chest_dagger_02(ComponentPose=@n114, Alpha=0.4, ExternalForce="0,0,0") {
    angularDampingOverride=5
    bGravityOverrideInSimSpace=true
    bUseGravityOverride=true
    alphaBoolBlend {blendOption=Linear}
    boundBone {boneName=dyn_chest_dagger_02}
    gravityOverride {z=-980}
    physicsBodyDefinitions [{
      sphereCollisionRadius=0
      boundBone {boneName=dyn_chest_dagger_02}
      boxExtents=<15 15 15>
      constraintSetup {twistAxis=AxisZ, angularLimitsMax=<4 4 0>, angularLimitsMin=<-4 -4 0>, angularTarget=<0 0 0>}
      localJointOffset=<0 0 -1>
     }]
    relativeSpaceBone {boneName=root}
   }
   n116 = Component To Local(ComponentPose=@n115)
   n117 = Get bIsInFrontEnd()
   n118 = Anim Dynamics - Bone: dyn_potion(ComponentPose=Local To Component(LocalPose=Copy Pose From Mesh() {bUseAttachedParent=true}), Alpha=0.5, ExternalForce="0,0,0") {
    angularDampingOverride=1
    angularSpringConstant=20000
    bAngularSpring=true
    bOverrideAngularDamping=true
    bUseGravityOverride=true
    gravityScale=0.35
    simulationSpace=RootRelative
    alphaBoolBlend {blendOption=Linear}
    boundBone {boneName=dyn_potion}
    componentLinearAccScale=<1 1 1>
    componentLinearVelScale=<1 1 1>
    gravityOverride {z=-980}
    physicsBodyDefinitions [{
      boundBone {boneName=dyn_potion}
      boxExtents=<15 10 10>
      constraintSetup {angularLimitsMax=<10 5 15>, angularLimitsMin=<-10 -10 -5>, angularTarget=<0 0 0>}
      localJointOffset=<5 0 0>
     }]
    relativeSpaceBone {boneName=spine_05}
   }
   n119 = Anim Dynamics - Bone: dyn_bag_r(ComponentPose=@n118, Alpha=0.5, ExternalForce="0,0,0") {
    angularDampingOverride=1
    angularSpringConstant=10000
    bAngularSpring=true
    bOverrideAngularDamping=true
    bUseGravityOverride=true
    gravityScale=0.35
    simulationSpace=RootRelative
    alphaBoolBlend {blendOption=Linear}
    boundBone {boneName=dyn_bag_r}
    componentLinearAccScale=<1 1 1>
    componentLinearVelScale=<1 1 1>
    gravityOverride {z=-980}
    physicsBodyDefinitions [{
      boundBone {boneName=dyn_bag_r}
      boxExtents=<15 15 15>
      constraintSetup {angularLimitsMax=<5 10 2>, angularLimitsMin=<-5 -15 -2>, angularTarget=<0 0 0>}
      localJointOffset=<3 0 0>
     }]
    relativeSpaceBone {boneName=pelvis}
   }
   n120 = Anim Dynamics - Bone: dyn_bag_l(ComponentPose=@n119, Alpha=0.4, ExternalForce="0,0,0") {
    angularDampingOverride=0.9999
    angularSpringConstant=5000
    bAngularSpring=true
    bOverrideAngularDamping=true
    bUseGravityOverride=true
    gravityScale=0.35
    simulationSpace=RootRelative
    alphaBoolBlend {blendOption=Linear}
    boundBone {boneName=dyn_bag_l}
    componentLinearAccScale=<1 1 1>
    componentLinearVelScale=<1 1 1>
    gravityOverride {z=-980}
    physicsBodyDefinitions [{
      boundBone {boneName=dyn_bag_l}
      boxExtents=<10 10 10>
      constraintSetup {angularLimitsMax=<1 10 1>, angularLimitsMin=<-1 -10 -1>, angularTarget=<0 0 0>}
      localJointOffset=<5 0 0>
     }]
   }
   n121 = Anim Dynamics - Bone: dyn_belt_knife(ComponentPose=@n120, Alpha=0.3, ExternalForce="0,0,0") {
    angularDampingOverride=1
    angularSpringConstant=20000
    bAngularSpring=true
    bOverrideAngularDamping=true
    bUseGravityOverride=true
    gravityScale=0.35
    simulationSpace=RootRelative
    alphaBoolBlend {blendOption=Linear}
    boundBone {boneName=dyn_belt_knife}
    componentLinearAccScale=<1 1 1>
    componentLinearVelScale=<1 1 1>
    gravityOverride {z=-980}
    physicsBodyDefinitions [{
      boundBone {boneName=dyn_belt_knife}
      boxExtents=<10 10 10>
      constraintSetup {angularLimitsMax=<0 1 5>, angularLimitsMin=<0 -10 -5>, angularTarget=<0 0 0>}
      localJointOffset=<5 0 0>
     }]
    relativeSpaceBone {boneName=pelvis}
   }
   n122 = "Transform (Modify) Bone - Bone: dyn_belt_knife"(ComponentPose=@n121, Alpha=1, Translation="0.5,0,0", Rotation="5,0,0", Scale="1,1,1") {
    rotationMode=BMM_Additive
    translationMode=BMM_Additive
    alphaBoolBlend {blendOption=Linear}
    boneToModify {boneName=dyn_belt_knife}
    rotation {pitch=5}
    translation {x=0.5}
   }
   n123 = "Constraint - dyn_skirt_front_01_l (P)"(ComponentPose=@n122, Alpha=0.8, ConstraintWeights_0=1) {
    alphaBoolBlend {blendOption=Linear}
    boneToModify {boneName=dyn_skirt_front_01_l}
    constraintSetup [{targetBone {boneName=thigh_l}}]
    constraintWeights [1]
   }
   n124 = "Constraint - dyn_skirt_side_04_r (P)"(ComponentPose=@n123, Alpha=1, ConstraintWeights_0=1, ConstraintWeights_1=0) {
    alphaBoolBlend {blendOption=Linear}
    boneToModify {boneName=dyn_skirt_side_04_r}
    constraintSetup [{targetBone {boneName=dyn_dagger}}]
    constraintWeights [1,0]
   }
   n125 = "Constraint - dyn_skirt_front_04_r (PPPPPP)"(
    ComponentPose=@n124
    Alpha=1
    ConstraintWeights_0=0.0001
    ConstraintWeights_1=0.4781
    ConstraintWeights_2=0
    ConstraintWeights_3=0.0003
    ConstraintWeights_4=0.5194
    ConstraintWeights_5=0.0021
   ) {
    alphaBoolBlend {blendOption=Linear}
    boneToModify {boneName=dyn_skirt_front_04_r}
    constraintSetup [
     {targetBone {boneName=spine_01}}
     {targetBone {boneName=thigh_r}}
     {targetBone {boneName=calf_twist_02_r}}
     {targetBone {boneName=calf_r}}
     {targetBone {boneName=thigh_twist_01_r}}
     {targetBone {boneName=pelvis}}
    ]
    constraintWeights [1,1,1,1,1,1]
   }
   n126 = "Constraint - dyn_skirt_front_02_r (PPPPPP)"(
    ComponentPose=@n125
    Alpha=1
    ConstraintWeights_0=0.0287
    ConstraintWeights_1=0.0225
    ConstraintWeights_2=0.0197
    ConstraintWeights_3=0.6174
    ConstraintWeights_4=0.04
    ConstraintWeights_5=0.2717
   ) {
    alphaBoolBlend {blendOption=Linear}
    boneToModify {boneName=dyn_skirt_front_02_r}
    constraintSetup [
     {targetBone {boneName=spine_01}}
     {targetBone {boneName=dyn_skirt_front_01_r}}
     {targetBone {boneName=thigh_r}}
     {targetBone {boneName=thigh_twist_01_r}}
     {targetBone {boneName=thigh_twist_01_l}}
     {targetBone {boneName=pelvis}}
    ]
    constraintWeights [1,1,1,1,1,1]
   }
   n127 = "Constraint - dyn_skirt_front_03_r (PPPP)"(ComponentPose=@n126, Alpha=1, ConstraintWeights_0=0.0054, ConstraintWeights_1=0.1072, ConstraintWeights_2=0.8096, ConstraintWeights_3=0.0777) {
    alphaBoolBlend {blendOption=Linear}
    boneToModify {boneName=dyn_skirt_front_03_r}
    constraintSetup [{targetBone {boneName=spine_01}}, {targetBone {boneName=thigh_r}}, {targetBone {boneName=thigh_twist_01_r}}, {targetBone {boneName=pelvis}}]
    constraintWeights [1,1,1,1]
   }
   n128 = "Constraint - dyn_skirt_front_01_r (P)"(ComponentPose=@n127, Alpha=6, ConstraintWeights_0=1) {
    alphaBoolBlend {blendOption=Linear}
    boneToModify {boneName=dyn_skirt_front_01_r}
    constraintSetup [{targetBone {boneName=thigh_r}}]
    constraintWeights [1]
   }
   n129 = RigidBody(ComponentPose=@n128, ExternalForce=Get GravityOverride()) {
    bClampLinearTranslationLimitToRefPose=true
    overridePhysicsAsset=&M_MED_Swamp_Knight_Crouch_Physics
    simulationSpace=BaseBoneSpace
    alphaBoolBlend {blendOption=Linear}
    baseBoneRef {boneName=pelvis}
    componentAppliedLinearAccClamp=<2750 2750 2750>
    componentLinearAccScale=<0.25 0.25 0.15>
    componentLinearVelScale=<0.25 0.25 0.15>
    overrideWorldGravity {z=-980}
   }
   n130 = "Transform (Modify) Bone - Bone: dyn_skirt_front_03_r"(ComponentPose=@n129, Alpha=1, Translation="0,0,4", Rotation="0,0,0", Scale="1,1,1") {translationMode=BMM_Additive, alphaBoolBlend {blendOption=Linear}, boneToModify {boneName=dyn_skirt_front_03_r}, translation {z=4}}
   n131 = "Transform (Modify) Bone - Bone: dyn_bag"(ComponentPose=@n130, Alpha=1, Translation="-2,2,4", Rotation="0,0,0", Scale="1,1,1") {
    rotationMode=BMM_Additive
    translationMode=BMM_Additive
    alphaBoolBlend {blendOption=Linear}
    boneToModify {boneName=dyn_bag}
    translation=<-2 2 4>
   }
   n132 = "Constraint - dyn_chest_dagger_00_base (PPPPPPPP)"(
    ComponentPose=@n131
    Alpha=1
    ConstraintWeights_0=0.006
    ConstraintWeights_1=0.0476
    ConstraintWeights_2=0.187
    ConstraintWeights_3=0.3862
    ConstraintWeights_4=0.2588
    ConstraintWeights_5=0.0289
    ConstraintWeights_6=0.0831
    ConstraintWeights_7=0.0022
   ) {
    alphaBoolBlend {blendOption=Linear}
    boneToModify {boneName=dyn_chest_dagger_00_base}
    constraintSetup [
     {targetBone {boneName=clavicle_l}}
     {targetBone {boneName=spine_01}}
     {targetBone {boneName=spine_02}}
     {targetBone {boneName=spine_03}}
     {targetBone {boneName=spine_04}}
     {targetBone {boneName=spine_05}}
     {targetBone {boneName=clavicle_r}}
     {targetBone {boneName=pelvis}}
    ]
    constraintWeights [1,1,1,1,1,1,1,1]
   }
   n133 = "Constraint - dyn_chest_dagger_01_base (PPPPPPPP)"(
    ComponentPose=@n132
    Alpha=1
    ConstraintWeights_0=0.1183
    ConstraintWeights_1=0.0377
    ConstraintWeights_2=0.1683
    ConstraintWeights_3=0.3301
    ConstraintWeights_4=0.2998
    ConstraintWeights_5=0.0458
    ConstraintWeights_6=0
    ConstraintWeights_7=0
   ) {
    alphaBoolBlend {blendOption=Linear}
    boneToModify {boneName=dyn_chest_dagger_01_base}
    constraintSetup [
     {targetBone {boneName=clavicle_l}}
     {targetBone {boneName=spine_01}}
     {targetBone {boneName=spine_02}}
     {targetBone {boneName=spine_03}}
     {targetBone {boneName=spine_04}}
     {targetBone {boneName=spine_05}}
     {targetBone {boneName=clavicle_r}}
     {targetBone {boneName=pelvis}}
    ]
    constraintWeights [1,1,1,1,1,1,1,1]
   }
   n134 = "Constraint - dyn_chest_dagger_02_base (PPPPPPPP)"(
    ComponentPose=@n133
    Alpha=1
    ConstraintWeights_0=0.1681
    ConstraintWeights_1=0.0351
    ConstraintWeights_2=0.1477
    ConstraintWeights_3=0.2823
    ConstraintWeights_4=0.3184
    ConstraintWeights_5=0.0519
    ConstraintWeights_6=0
    ConstraintWeights_7=0.0001
   ) {
    alphaBoolBlend {blendOption=Linear}
    boneToModify {boneName=dyn_chest_dagger_02_base}
    constraintSetup [
     {targetBone {boneName=clavicle_l}}
     {targetBone {boneName=spine_01}}
     {targetBone {boneName=spine_02}}
     {targetBone {boneName=spine_03}}
     {targetBone {boneName=spine_04}}
     {targetBone {boneName=spine_05}}
     {targetBone {boneName=clavicle_r}}
     {targetBone {boneName=pelvis}}
    ]
    constraintWeights [1,1,1,1,1,1,1,1]
   }
   n135 = Anim Dynamics - Bone: dyn_chest_dagger_00(ComponentPose=@n134, Alpha=0.6, ExternalForce="0,0,0") {
    angularDampingOverride=5
    bGravityOverrideInSimSpace=true
    bUseGravityOverride=true
    alphaBoolBlend {blendOption=Linear}
    boundBone {boneName=dyn_chest_dagger_00}
    gravityOverride {z=-980}
    physicsBodyDefinitions [{
      sphereCollisionRadius=0
      boundBone {boneName=dyn_chest_dagger_00}
      boxExtents=<15 15 15>
      constraintSetup {twistAxis=AxisZ, angularLimitsMax=<1 1 4>, angularLimitsMin=<-1 -1 -4>, angularTarget=<0 0 0>}
      localJointOffset=<0 0 -1>
     }]
    relativeSpaceBone {boneName=root}
   }
   n136 = Anim Dynamics - Bone: dyn_chest_dagger_01(ComponentPose=@n135, Alpha=0.5, ExternalForce="0,0,0") {
    angularDampingOverride=5
    bGravityOverrideInSimSpace=true
    bUseGravityOverride=true
    alphaBoolBlend {blendOption=Linear}
    boundBone {boneName=dyn_chest_dagger_01}
    gravityOverride {z=-980}
    physicsBodyDefinitions [{
      sphereCollisionRadius=0
      boundBone {boneName=dyn_chest_dagger_01}
      boxExtents=<13 13 13>
      constraintSetup {twistAxis=AxisZ, angularLimitsMax=<4 4 0>, angularLimitsMin=<-4 -4 0>, angularTarget=<0 0 0>}
      localJointOffset=<0 0 -1>
     }]
    relativeSpaceBone {boneName=root}
   }
   n137 = Anim Dynamics - Bone: dyn_chest_dagger_02(ComponentPose=@n136, Alpha=0.4, ExternalForce="0,0,0") {
    angularDampingOverride=5
    bGravityOverrideInSimSpace=true
    bUseGravityOverride=true
    alphaBoolBlend {blendOption=Linear}
    boundBone {boneName=dyn_chest_dagger_02}
    gravityOverride {z=-980}
    physicsBodyDefinitions [{
      sphereCollisionRadius=0
      boundBone {boneName=dyn_chest_dagger_02}
      boxExtents=<15 15 15>
      constraintSetup {twistAxis=AxisZ, angularLimitsMax=<4 4 0>, angularLimitsMin=<-4 -4 0>, angularTarget=<0 0 0>}
      localJointOffset=<0 0 -1>
     }]
    relativeSpaceBone {boneName=root}
   }
   n138 = Control Rig!(
    Source=Blend Poses by bool(
     BlendPose_0=Component To Local(ComponentPose=@n137)
     BlendPose_1=Blend Poses by bool(BlendPose_0=@n116, BlendPose_1=@n97, BlendTime_0=0.3, BlendTime_1=0.5, bActiveValue=@n117) {blendType=Linear}
     BlendTime_0=0.3
     BlendTime_1=0.5
     bActiveValue=Get bIsPlayingEmote()
    ) {blendType=Linear}
    Alpha
    Is Sky Diving_1:bool=Get Is Sky Diving() -> (Is Sky Diving:bool)
    CrouchTarget:bool=Get CrouchTarget() -> (CrouchTarget:bool)
    SkydiveFloat_0:bool=Get SkydiveFloat() -> (SkydiveFloat:bool)
    BaseBall_Slide:double=Get Baseball_Slide() -> (Baseball_Slide:double)
    Knee_Slide:double=Get Knee_Slide() -> (Knee_Slide:double)
   ) -> (Pose) {
    alphaBoolBlend {blendOption=Linear}
    controlRigAssetReference {blueprintRigClass=&/Swamp_Knight_Commando_CtrlRig.Swamp_Knight_Commando_CtrlRig_C}
   }
   Output Pose(Result=@n138) {layerGroup=None}
   Get bIsPlayingEmote()
   Get CrouchOnly() -> (CrouchOnly:bool)
  ]}
 EventGraph {nodes [
   Event BlueprintUpdateAnimation()
   > n1 = Sequence!(execute) -> (then_0, then_1, then_2:Exec, then_3:Exec, then_4:Exec, then_5:Exec, then_6:Exec)
   n2 = Get bIsMoving2D()
   n3 = Branch(execute=@n1.then_1, Condition=AND Boolean(A=Get bIsCrouching(), B=NOT Boolean(A=@n2)))
   > Set CrouchOnly(CrouchOnly:bool=true) -> (Output_Get:bool)
   Set CrouchOnly(execute=@n3.else, CrouchOnly=false)
   Set Is Targeting(execute=@n1.then_2, Is Targeting:bool=Get bIsTargeting()) -> (Output_Get:bool)
   > Set Is Relaxed Level 1(Is Relaxed Level 1:bool=Get bIsRelaxedLevel1() -> (bIsRelaxedLevel1:bool)) -> (Output_Get:bool)
   > Set Is Sky Diving(Is Sky Diving:bool=Get bIsSkyDiving()) -> (Output_Get:bool)
   n4 = Gravity Override State Function DEPRECATED(
    execute=@n1.then_0
    CharacterPartAnimInstance=Self-Reference()
    GameTimeInSeconds=GetGameTimeInSeconds()
    GravityOverrideParams=Get Gravity Override Params()
   )
   > Set GravityOverride(GravityOverride=@n4)
   Set CrouchTarget(execute=@n1.then_3, CrouchTarget:bool=AND Boolean(A=Get bIsCrouching(), B=Get bIsTargeting())) -> (Output_Get:bool)
   Set DiveOrFloat(execute=@n1.then_4, DiveOrFloat:bool=OR Boolean(A=Get bIsSkydiveDiveMode(), B=Get bIsSkydiveFloating())) -> (Output_Get:bool)
   Set SkydiveFloat(execute=@n1.then_5, SkydiveFloat:bool=Get bIsSkydiveFloating()) -> (Output_Get:bool)
   n5 = Branch(execute=@n1.then_6, Condition=Get bIsSliding())
   > n6 = Set Left_Calf(Left_Calf:double=GetAlphaFromReferenceTransformDelta(SkeletalMeshComponent=GetPartSkeletalMeshComponent(), BoneName=calf_l, BaseName=thigh_l, InputRange=(X=0,Y=1), TwistAxis="1,0,0").RawOutput) -> (Output_Get:double)
   > n7 = Branch(Condition=NOT Boolean(A="float > float"(A=@n6, B=100)))
   > Set Knee_Slide(Knee_Slide:double=0) -> (Output_Get:double)
   > Set Baseball_Slide(Baseball_Slide:double=1) -> (Output_Get:double)
   Set Knee_Slide(execute=@n7.else, Knee_Slide=1)
   > Set Baseball_Slide(Baseball_Slide=0)
   Set Knee_Slide(execute=@n5.else, Knee_Slide=0)
   > Set Baseball_Slide(Baseball_Slide=0)
  ]}
]
variables [
 Baseball_Slide {type=number}
 CrouchOnly {type=boolean}
 CrouchTarget {type=boolean}
 DiveOrFloat {type=boolean}
 Gravity Override Params {type=GravityOverrideParamsStruct, default {skydive_DivingDown {gravityOverride=<-2500 0 0>}}}
 GravityOverride {type=Vector}
 Is Relaxed Level 1 {type=boolean}
 Is Sky Diving {type=boolean}
 Is Targeting {type=boolean}
 Knee_Slide {type=number}
 Left_Calf {type=number}
 PickaxeEquip {type=boolean}
 SkydiveFloat {type=boolean}
]
